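# Console profile widget that shows a "Signed in as:" label beside the profile's display name.
# A padded CommonBorder holds a horizontal box with the prefix text in regular style and the name in bold.
# Both text blocks are inherited from the native FortConsoleProfileWidget parent, and the event graph is empty, so this Blueprint defines only layout and styling.
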WidgetBlueprint /FortniteCoreUI/UI/Foundation/Profile/ConsoleProfileWidget
graphs [EventGraph {links [], nodes []}]
variables []
widgets {
 info {
  inheritedWidgetCount=2
  parentClass=/Script/FortniteUI.FortConsoleProfileWidget
  rootWidgetClass=&CommonBorder
  widgetCount=4
 }
 widgets [
  CommonBorder_0 {widgetClassPath=&CommonBorder, widget=CommonBorder_0, properties {padding=<ltrb 8 8 8 8>}}
  HorizontalBox_0 {
   widgetClassPath=&HorizontalBox
   parent=CommonBorder_0
   widget=HorizontalBox_0
   slot {class=BorderSlot, properties {padding=<ltrb 8 8 8 8>}}
  }
  Text_Prefix {
   widgetClassPath=&CommonTextBlock
   bInherited=true
   bIsVariable=true
   parent=HorizontalBox_0
   widget=Text_Prefix
   properties {
    style=/FortUILibrary/Styles/Typography/Foundation/TextStyle-Base-S.TextStyle-Base-S_C
    text=Signed in as:
    font {
     bForceMonospaced=false
     bMaterialIsStencil=false
     fontMaterial=null
     fontObject=&NotoSans
     letterSpacing=0
     monospacedWidth=1
     size=18
     skewAmount=0
     typefaceFontName=NotoSans-Regular
     outlineSettings {
      bApplyOutlineToDropShadows=false
      bMiteredCorners=false
      bSeparateFillAlpha=false
      outlineMaterial=null
      outlineSize=0
      outlineColor=<rgba 0 0 0 1>
     }
    }
    shadowOffset=<0 0>
   }
   slot {class=HorizontalBoxSlot, properties {padding=<ltrb 0 0 12 0>}}
  }
  Text_DisplayName {
   widgetClassPath=&CommonTextBlock
   bInherited=true
   bIsVariable=true
   parent=HorizontalBox_0
   widget=Text_DisplayName
   properties {
    style=/FortUILibrary/Styles/Typography/Foundation/TextStyle-Base-S-B.TextStyle-Base-S-B_C
    text=Profile Name
    font {
     bForceMonospaced=false
     bMaterialIsStencil=false
     fontMaterial=null
     fontObject=&NotoSans
     letterSpacing=0
     monospacedWidth=1
     size=18
     skewAmount=0
     typefaceFontName=NotoSans-Bold
     outlineSettings {
      bApplyOutlineToDropShadows=false
      bMiteredCorners=false
      bSeparateFillAlpha=false
      outlineMaterial=null
      outlineSize=0
      outlineColor=<rgba 0 0 0 1>
     }
    }
    shadowOffset=<0 0>
   }
   slot {class=HorizontalBoxSlot}
  }
 ]
}
