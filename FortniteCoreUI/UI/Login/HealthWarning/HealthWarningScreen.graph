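# Login health warning screen widget, derived from the native HealthWarningScreen class, showing centered warning text and restricted-country rating icons.
# Both inherited widgets sit in an overlay inside a safe zone and border.
# A HandleShowTimerComplete event broadcasts the HealthWarningCompleted dispatcher; this graph neither sets the ShowTimer handle nor calls the event.
# FinishingAnim fades the border content out, but nothing in this graph plays it.
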WidgetBlueprint /FortniteCoreUI/UI/Login/HealthWarning/HealthWarningScreen
animations [FinishingAnim {
  binding_count=1
  display_label=""
  length_seconds=1
  object=&self:FinishingAnim
  sequence {bindings [CommonBorder_0 {
     id=#1
     possessed_class=CommonBorder
     tracks [{
       type=MovieSceneColorTrack
       display_name=Content Color and Opacity
       sections [{
         type=MovieSceneColorSection
         end_frame=20
         start_frame=0
         channels [
          Color.R {type=MovieSceneScriptingFloatChannel, const [1,1]}
          Color.G {type=MovieSceneScriptingFloatChannel, const [1,1]}
          Color.B {type=MovieSceneScriptingFloatChannel, const [1,1]}
          Color.A {type=MovieSceneScriptingFloatChannel, keys [[0,1],[20,0]]}
         ]
        }]
      }]
    }]}
 }]
graphs [
 EventGraph {nodes [
   HandleShowTimerComplete() {is=title K2Node_CustomEvent}
   > Call Health Warning Completed!(execute:Exec, self:Health Warning Screen Object Reference) -> (then:Exec) {is=title K2Node_CallDelegate}
  ]}
 HealthWarningCompleted {nodes [HealthWarningCompleted() {is=title K2Node_FunctionEntry}]}
]
variables [
 HealthWarningCompleted {}
 IconPadding {type=Margin, default=<ltrb 10 10 10 10>}
 ShowTimer {type=TimerHandle}
]
widgets {
 info {inheritedWidgetCount=2, parentClass=/Script/FortniteUI.HealthWarningScreen, rootWidgetClass=&SafeZone, widgetCount=5}
 widgets [
  SafeZone_0 {widgetClassPath=&SafeZone, bIsVariable=true, widget=SafeZone_0}
  CommonBorder_0 {
   widgetClassPath=&CommonBorder
   parent=SafeZone_0
   widget=CommonBorder_0
   properties {padding=<ltrb 0 0 0 0>}
   slot {class=SafeZoneSlot}
  }
  Overlay_0 {widgetClassPath=&Overlay, parent=CommonBorder_0, widget=Overlay_0, slot {class=BorderSlot, properties {padding=<ltrb 0 0 0 0>}}}
  Text_HealthWarning {
   widgetClassPath=&CommonTextBlock
   bInherited=true
   bIsVariable=true
   parent=Overlay_0
   widget=Text_HealthWarning
   properties {
    autoWrapText=true
    justification=Center
    style=/FortUILibrary/Styles/Typography/Foundation/TextStyle-Header-L.TextStyle-Header-L_C
    font {
     bForceMonospaced=false
     bMaterialIsStencil=false
     fontMaterial=null
     fontObject=&Burbank
     letterSpacing=0
     monospacedWidth=1
     size=48
     skewAmount=0
     typefaceFontName=Burbank-Black
     outlineSettings {
      bApplyOutlineToDropShadows=false
      bMiteredCorners=false
      bSeparateFillAlpha=false
      outlineMaterial=null
      outlineSize=0
      outlineColor=<rgba 0 0 0 1>
     }
    }
    shadowOffset=<0 0>
   }
   slot {class=OverlaySlot, properties {horizontalAlignment=HAlign_Center, verticalAlignment=VAlign_Center, padding=<ltrb 300 0 300 0>}}
  }
  RatingIcons_RatingIcons {
   widgetClassPath=/Game/UI/Frontend/Misc/RestrictedCountryRatingIcons
   bInherited=true
   bIsVariable=true
   parent=Overlay_0
   widget=RatingIcons_RatingIcons
   slot {class=OverlaySlot, properties {horizontalAlignment=HAlign_Right, padding=<ltrb 0 300 240 0>}}
  }
 ]
}
