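# Login interaction host widget, a child of FortLoginInteractions, that holds the login flow's screens.
# An inherited activatable switcher stacks account selection, sign-in, display name, MFA, EULA, account linking, web login, age gate and corrective-action panels behind a default spacer.
# A safe-zone corner back button overlays the switcher.
# The graph only reports back-button clicks through Fire Event Element Interacted; screen switching is not in this Blueprint.
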WidgetBlueprint /FortniteCoreUI/UI/Login/LoginInteractionHost
graphs [EventGraph {nodes [
   n1 = "On Button Clicked (Button_Back)"!() -> (OutputDelegate, then, Button:Common Button Legacy Object Reference) {is=title K2Node_ComponentBoundEvent}
   > Fire Event Element Interacted(Element=@n1.Button)
   Event PreConstruct()
   Event Construct()
   Event Tick()
  ]}]
variables []
widgets {
 info {inheritedWidgetCount=15, parentClass=/Script/FortniteUI.FortLoginInteractions, rootWidgetClass=&ScaleBox, widgetCount=18}
 widgets [
  ScaleBox_0 {widgetClassPath=&ScaleBox, widget=ScaleBox_0, properties {stretch=ScaleBySafeZone}}
  Overlay_127 {
   widgetClassPath=&Overlay
   parent=ScaleBox_0
   widget=Overlay_127
   slot {class=ScaleBoxSlot, properties {horizontalAlignment=HAlign_Fill, verticalAlignment=VAlign_Fill}}
  }
  Switcher_Input {
   widgetClassPath=&CommonActivatableWidgetSwitcher
   bInherited=true
   bIsVariable=true
   parent=Overlay_127
   widget=Switcher_Input
   slot {class=OverlaySlot, properties {horizontalAlignment=HAlign_Fill, verticalAlignment=VAlign_Fill}}
  }
  Spacer_Default {
   widgetClassPath=&Spacer
   bInherited=true
   parent=Switcher_Input
   widget=Spacer_Default
   slot {class=WidgetSwitcherSlot, properties {horizontalAlignment=HAlign_Left, verticalAlignment=VAlign_Top}}
  }
  Container_StartLogin {
   widgetClassPath=/FortniteCoreUI/UI/Login/AccountSelect/PlayedBeforeSelectContainer
   bInherited=true
   bIsVariable=true
   parent=Switcher_Input
   widget=Container_StartLogin
   slot {class=WidgetSwitcherSlot}
  }
  Credential_Select {
   widgetClassPath=/FortniteCoreUI/UI/Login/AccountSelect/LoginCredentialSelect
   bInherited=true
   bIsVariable=true
   parent=Switcher_Input
   widget=Credential_Select
   properties {bIsFocusable=true}
   slot {class=WidgetSwitcherSlot}
  }
  SignIn_Credentials {
   widgetClassPath=/FortniteCoreUI/UI/Login/SignIn/SignInWidget
   bInherited=true
   bIsVariable=true
   parent=Switcher_Input
   widget=SignIn_Credentials
   properties {bIsFocusable=true}
   slot {class=WidgetSwitcherSlot}
  }
  Account_NotFound {
   widgetClassPath=/FortniteCoreUI/UI/Login/AccountSelect/AccountNotFound
   bInherited=true
   bIsVariable=true
   parent=Switcher_Input
   widget=Account_NotFound
   properties {bIsFocusable=true}
   slot {class=WidgetSwitcherSlot}
  }
  DisplayName_Create {
   widgetClassPath=/FortniteCoreUI/UI/Login/DisplayName/DisplayName
   bInherited=true
   bIsVariable=true
   parent=Switcher_Input
   widget=DisplayName_Create
   properties {bIsFocusable=true}
   slot {class=WidgetSwitcherSlot}
  }
  MFA_Auth {
   widgetClassPath=/FortniteCoreUI/UI/Login/MFA/MultiFactorAuthWidget
   bInherited=true
   bIsVariable=true
   parent=Switcher_Input
   widget=MFA_Auth
   properties {bIsFocusable=true}
   slot {class=WidgetSwitcherSlot}
  }
  Container_EulaWidgets {
   widgetClassPath=/FortniteCoreUI/UI/Login/Eula/EulaWidgetContainer
   bInherited=true
   bIsVariable=true
   parent=Switcher_Input
   widget=Container_EulaWidgets
   slot {class=WidgetSwitcherSlot}
  }
  Account_Link {
   widgetClassPath=/FortniteCoreUI/UI/Login/AccountLinking/AccountLinkingWindow
   bInherited=true
   bIsVariable=true
   parent=Switcher_Input
   widget=Account_Link
   properties {bIsFocusable=true}
   slot {class=WidgetSwitcherSlot}
  }
  Account_PinLink {
   widgetClassPath=/FortniteCoreUI/UI/Login/AccountLinking/AccountPinLinkingWindow
   bInherited=true
   bIsVariable=true
   parent=Switcher_Input
   widget=Account_PinLink
   slot {class=WidgetSwitcherSlot}
  }
  Web_Login {
   widgetClassPath=/FortniteCoreUI/Frontend/StoreMain/WebLogin
   bInherited=true
   bIsVariable=true
   parent=Switcher_Input
   widget=Web_Login
   slot {class=WidgetSwitcherSlot}
  }
  AgeGateScreen {
   widgetClassPath=/FortniteCoreUI/UI/Login/CorrectiveAction/CorrectiveActionDateOfBirthScreen
   bInherited=true
   bIsVariable=true
   parent=Switcher_Input
   widget=AgeGateScreen
   slot {class=WidgetSwitcherSlot}
  }
  CorrectiveActionFlow {
   widgetClassPath=/FortniteCoreUI/UI/Login/CorrectiveAction/CorrectiveActionFlow
   bInherited=true
   bIsVariable=true
   parent=Switcher_Input
   widget=CorrectiveActionFlow
   slot {class=WidgetSwitcherSlot}
  }
  SafeZone_0 {widgetClassPath=&SafeZone, bIsVariable=true, parent=Overlay_127, widget=SafeZone_0, slot {class=OverlaySlot}}
  Button_Back {
   widgetClassPath=&IconTextButton
   bInherited=true
   bIsVariable=true
   parent=SafeZone_0
   widget=Button_Back
   properties {
    buttonText=""
    onButtonBaseClicked=(&self:WidgetTree.Button_Back.HandleOnButtonClicked)
    onButtonBaseDoubleClicked=(&self:WidgetTree.Button_Back.HandleOnButtonDoubleClicked)
    onButtonBaseHovered=(&self:WidgetTree.Button_Back.HandleOnButtonHovered)
    onButtonBaseUnhovered=(&self:WidgetTree.Button_Back.HandleOnButtonUnhovered)
    onSelectedChangedBase=(&self:WidgetTree.Button_Back.HandleOnSelectedChanged)
    style=/FortUILibrary/Foundation/Buttons/ButtonStyles/ButtonStyle-TextOnlyBase_Empty.ButtonStyle-TextOnlyBase_Empty_C
    iconBrush {
     bIsDynamicallyLoaded=false
     drawAs=Image
     imageType=NoImage
     mirroring=NoMirror
     resourceName=None
     resourceObject=/FortniteCoreUI/UI/Login/Art/M_CornerBackArrow
     tiling=NoTile
     imageSize=<112 112>
     margin=<ltrb 0 0 0 0>
     outlineSettings {
      bUseBrushTransparency=false
      roundingType=HalfHeightRadius
      width=0
      color {colorUseRule=UseColor_Specified, specifiedColor=<rgba 0 0 0 0>}
      cornerRadii=<0 0 0 1>
     }
     tintColor {colorUseRule=UseColor_Specified, specifiedColor=<rgba 1 1 1 1>}
     uVRegion {bIsValid=false, max=<0 0>, min=<0 0>}
    }
   }
   slot {class=SafeZoneSlot}
  }
 ]
}
