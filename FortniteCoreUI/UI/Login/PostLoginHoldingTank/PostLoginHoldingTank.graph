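# Widget Blueprint for the post-login holding tank screen, shown while Fortnite downloads and installs content, built on native FortPostLoginHoldingTankScreen.
# An overlay stacks a progression widget with an activatable switcher holding error, action and status modals: connection, disk space, cellular, notifications, download confirmation and background or foreground progress.
# Each modal is a configured instance with its own text, error code and placeholders, some as iOS variants.
# Widgets bind to inherited variables; the event graph has no logic.
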WidgetBlueprint /FortniteCoreUI/UI/Login/PostLoginHoldingTank/PostLoginHoldingTank
graphs [EventGraph {nodes [
   Event PreConstruct()
   Event Construct()
   Event Tick()
  ]}]
variables []
widgets {
 info {inheritedWidgetCount=19, parentClass=/Script/FortniteUI.FortPostLoginHoldingTankScreen, rootWidgetClass=&Overlay, widgetCount=20}
 widgets [
  Overlay_0 {widgetClassPath=&Overlay, widget=Overlay_0}
  ProgressionWidget {
   widgetClassPath=/FortniteCoreUI/UI/Login/PostLoginHoldingTank/ProgressBar/HoldingTankProgressionWidget
   bInherited=true
   bIsVariable=true
   parent=Overlay_0
   widget=ProgressionWidget
   slot {class=OverlaySlot, properties {horizontalAlignment=HAlign_Fill, verticalAlignment=VAlign_Fill}}
  }
  ModalsSwitcher {
   widgetClassPath=&CommonActivatableWidgetSwitcher
   bInherited=true
   bIsVariable=true
   parent=Overlay_0
   widget=ModalsSwitcher
   slot {class=OverlaySlot, properties {horizontalAlignment=HAlign_Fill, verticalAlignment=VAlign_Fill}}
  }
  ConnectionErrorModal {
   widgetClassPath=&/HoldingTankErrorModal
   bInherited=true
   bIsVariable=true
   parent=ModalsSwitcher
   widget=ConnectionErrorModal
   properties {
    bodyText="An error occurred while connecting to Epic servers.\r\nPlease try again later."
    confirmButtonText=""
    declineButtonText=""
    errorCodeText=ESP-CON-000
    titleText=Error
   }
   slot {class=WidgetSwitcherSlot}
  }
  InstallerErrorModal {
   widgetClassPath=&HoldingTankErrorModal
   bInherited=true
   bIsVariable=true
   parent=ModalsSwitcher
   widget=InstallerErrorModal
   properties {
    bodyText="{InstallerErrorText}"
    confirmButtonText=Ok
    declineButtonText=""
    errorCodeText="ESP-INST-{InstallerErrorCode}"
    timerText=""
    titleText=Error
   }
   slot {class=WidgetSwitcherSlot}
  }
  OutOfDiskSpaceErrorModal {
   widgetClassPath=&HoldingTankErrorModal
   bInherited=true
   bIsVariable=true
   parent=ModalsSwitcher
   widget=OutOfDiskSpaceErrorModal
   properties {
    bodyText="There is not enough space on your device.\r\n{RequiredBytes} is required.\r\n{AvailableBytes} is available.\r\nYou need an additional {SpaceAdditional} to perform the installation."
    confirmButtonText=Retry
    declineButtonText=""
    errorCodeText=ESP-DISK-001
    timerText=""
    titleText=Error
   }
   slot {class=WidgetSwitcherSlot}
  }
  OutOfDiskSpaceErrorModal_IOS {
   widgetClassPath=&HoldingTankErrorModal
   bInherited=true
   bIsVariable=true
   parent=ModalsSwitcher
   widget=OutOfDiskSpaceErrorModal_IOS
   properties {
    bodyText="You need an additional {SpaceAdditional} to install Fortnite.\r\n\r\nFree up some room on your device and try again.\r\n\r\n1. Go to Settings\r\n\r\n2. Select General\r\n\r\n3. Select iPhone/iPad Storage\r\n\r\n4. Remove other apps or media\r\n"
    confirmButtonText=Try again
    declineButtonText=""
    errorCodeText=ESP-DISK-001
    timerText=""
    titleText=Not enough storage
   }
   slot {class=WidgetSwitcherSlot}
  }
  FailedCacheReserveErrorModal {
   widgetClassPath=&HoldingTankErrorModal
   bInherited=true
   bIsVariable=true
   parent=ModalsSwitcher
   widget=FailedCacheReserveErrorModal
   properties {
    bodyText=An error has occurred. Please try again later.
    confirmButtonText=""
    declineButtonText=""
    errorCodeText=ESP-CACHE-001
    titleText=Error
   }
   slot {class=WidgetSwitcherSlot}
  }
  MetadataErrorModal {
   widgetClassPath=&HoldingTankErrorModal
   bInherited=true
   bIsVariable=true
   parent=ModalsSwitcher
   widget=MetadataErrorModal
   properties {bodyText=An error has occurred. Please try again later., confirmButtonText="", declineButtonText="", errorCodeText=ESP-MET, titleText=Error}
   slot {class=WidgetSwitcherSlot}
  }
  UserPausedBundleInstallErrorModal {
   widgetClassPath=&HoldingTankErrorModal
   bInherited=true
   bIsVariable=true
   parent=ModalsSwitcher
   widget=UserPausedBundleInstallErrorModal
   properties {
    bodyText=The Install has been paused.
    confirmButtonText=Resume
    declineButtonText=""
    errorCodeText=ESP-PAUSE-001
    timerText=""
    titleText=Install Paused
   }
   slot {class=WidgetSwitcherSlot}
  }
  OnCellularErrorModal {
   widgetClassPath=&HoldingTankErrorModal
   bInherited=true
   bIsVariable=true
   parent=ModalsSwitcher
   widget=OnCellularErrorModal
   properties {
    bodyText="You are about to use cellular data to download Fortnite content. This will require downloading approximately {DownloadSizeEstimate}."
    confirmButtonText=Continue download
    declineButtonText=""
    errorCodeText=ESP-CELL-001
    timerText=""
    titleText=Warning
   }
   slot {class=WidgetSwitcherSlot}
  }
  NoInternetConnectionErrorModal {
   widgetClassPath=&HoldingTankErrorModal
   bInherited=true
   bIsVariable=true
   parent=ModalsSwitcher
   widget=NoInternetConnectionErrorModal
   properties {
    bodyText=No internet connection detected. Please check your connection and try again.
    confirmButtonText=""
    declineButtonText=""
    errorCodeText=ESP-CON-001
    titleText=Error
   }
   slot {class=WidgetSwitcherSlot}
  }
  EnableNotificationsErrorModal {
   widgetClassPath=&HoldingTankErrorModal
   bInherited=true
   bIsVariable=true
   parent=ModalsSwitcher
   widget=EnableNotificationsErrorModal
   properties {
    bodyText="We’ll let you know when you need to come back and finish installation.\r\n\r\nPlus, get notified about party invites, game updates, and more."
    confirmButtonText=GO TO SETTING
    declineButtonText=Not now
    errorCodeText=" "
    timerText=""
    titleText=ALLOW NOTIFICATIONS
   }
   slot {class=WidgetSwitcherSlot}
  }
  EnableNotificationsErrorModal_IOS {
   widgetClassPath=&HoldingTankErrorModal
   bInherited=true
   bIsVariable=true
   parent=ModalsSwitcher
   widget=EnableNotificationsErrorModal_IOS
   properties {
    bodyText="We’ll let you know when you need to come back and finish installation.\r\n\r\nPlus, get notified about party invites, game updates, and more."
    confirmButtonText=ALLOW NOTIFICATIONS
    declineButtonText=MAYBE LATER
    errorCodeText=" "
    timerText=""
    titleText=TURN ON NOTIFICATIONS
   }
   slot {class=WidgetSwitcherSlot}
  }
  EnableBackgroundDownloadErrorModal_IOS {
   widgetClassPath=&HoldingTankErrorModal
   bInherited=true
   bIsVariable=true
   parent=ModalsSwitcher
   widget=EnableBackgroundDownloadErrorModal_IOS
   properties {
    bodyText="This needs to be On for Fortnite to download in the background.\r\n\r\n1. Go to Settings.\r\n\r\n2. Set Background App Refresh to Wi-Fi or Wi-Fi && Cellular Data.\r\n\r\n3. Make sure Fortnite is toggled On.\r\n\r\n4. Return to Fortnite.\r\n"
    confirmButtonText=Go to Settings
    declineButtonText=No thanks
    errorCodeText=" "
    timerText=""
    titleText=Turn On background App refresh
   }
   slot {class=WidgetSwitcherSlot}
  }
  ConfirmDownloadModal {
   widgetClassPath=&/HoldingTankActionModal
   bInherited=true
   bIsVariable=true
   parent=ModalsSwitcher
   widget=ConfirmDownloadModal
   properties {bodyText="{DownloadSizeEstimate}", confirmButtonText=DOWNLOAD NOW, hideProgressBar=true, titleText=DOWNLOAD NOW}
   slot {class=WidgetSwitcherSlot}
  }
  ConfirmUpdateModal {
   widgetClassPath=&HoldingTankActionModal
   bInherited=true
   bIsVariable=true
   parent=ModalsSwitcher
   widget=ConfirmUpdateModal
   properties {bodyText="{DownloadSizeEstimate}", hideProgressBar=true}
   slot {class=WidgetSwitcherSlot}
  }
  ForegroundDownloadStatusModal {
   widgetClassPath=&/HoldingTankStatusModal
   bInherited=true
   bIsVariable=true
   parent=ModalsSwitcher
   widget=ForegroundDownloadStatusModal
   properties {actionButtonText=ENABLE BACKGROUND DOWNLOADS, bodyText=Keep Fortnite open to finish installing., titleText=Don't leave}
   slot {class=WidgetSwitcherSlot}
  }
  BackgroundDownloadStatusModal {
   widgetClassPath=&HoldingTankStatusModal
   bInherited=true
   bIsVariable=true
   parent=ModalsSwitcher
   widget=BackgroundDownloadStatusModal
   properties {
    actionButtonText=""
    bodyText="You can leave the app while it's downloading, and we'll notify you when it's time to come back."
    isAlert=false
    titleText=""
   }
   slot {class=WidgetSwitcherSlot}
  }
  ForegroundInstallStatusModal {
   widgetClassPath=&HoldingTankStatusModal
   bInherited=true
   bIsVariable=true
   parent=ModalsSwitcher
   widget=ForegroundInstallStatusModal
   properties {actionButtonText="", bodyText=Keep Fortnite open to finish installing., titleText=Don't leave}
   slot {class=WidgetSwitcherSlot}
  }
 ]
}
