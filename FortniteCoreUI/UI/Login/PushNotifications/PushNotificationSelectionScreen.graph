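# Login screen widget, built on FortPushNotificationSelectScreen, that explains push notifications and offers Continue and Not Now buttons.
# AdaptToScreenOrientation runs on Construct and on the game instance's orientation-change delegate, which is unbound on Destruct.
# It re-parents both button size boxes into a full-width vertical stack for portrait viewports with height/width above 1.34, otherwise into a left-aligned horizontal row.
# The graphs contain no button click handling.
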WidgetBlueprint /FortniteCoreUI/UI/Login/PushNotifications/PushNotificationSelectionScreen
graphs [
 AdaptToScreenOrientation {nodes [
   AdaptToScreenOrientation() {is=title K2Node_FunctionEntry}
   n1 = "Equal (Enum)"(
    A:EScreenOrientation Enum=GetDeviceOrientation!() -> (ReturnValue:EScreenOrientation Enum=Unknown)
    B:EScreenOrientation Enum=PortraitUpsideDown
   )
   n2 = "Equal (Enum)"(A=GetDeviceOrientation() -> (ReturnValue=Unknown), B=Portrait)
   n3 = GetViewportSize!() -> (ReturnValue_X:double, ReturnValue_Y:double)
   n4 = float / float(A=@n3.ReturnValue_Y, B=@n3)
   > n5 = Branch(Condition=AND Boolean(A=OR Boolean(A=@n2, B=@n1), B="float > float"(A=@n4, B=1.34)))
   n6 = Get VerticalBox_Buttons() -> (VerticalBox_Buttons:Vertical Box Object Reference)
   > AddChild(self=@n6, Content=Get Button_SizeBox_Continue() -> (Button_SizeBox_Continue:Size Box Object Reference))
   > AddChild(self=@n6, Content=Get Button_SizeBox_NotNow() -> (Button_SizeBox_NotNow:Size Box Object Reference))
   n7 = SlotAsVerticalBoxSlot(Widget=Get Button_SizeBox_Continue())
   > SetHorizontalAlignment(self:Vertical Box Slot Object Reference=@n7, InHorizontalAlignment=HAlign_Fill)
   > SetPadding!(
    execute
    self:Vertical Box Slot Object Reference=@n7
    InPadding_Left:float=40
    InPadding_Top:float=0
    InPadding_Right:float=40
    InPadding_Bottom:float=32
   ) -> (then)
   n8 = SlotAsVerticalBoxSlot(Widget=Get Button_SizeBox_NotNow())
   > SetHorizontalAlignment(self=@n8, InHorizontalAlignment=HAlign_Fill)
   > SetPadding(self=@n8, InPadding_Left=40, InPadding_Top=0, InPadding_Right=40, InPadding_Bottom=48)
   n9 = Get HorizontalBox_Buttons() -> (HorizontalBox_Buttons:Horizontal Box Object Reference)
   AddChildToHorizontalBox!(execute=@n5.else, self:Horizontal Box Object Reference=@n9, Content:Widget Object Reference=Get Button_SizeBox_Continue()) -> (then, ReturnValue:Horizontal Box Slot Object Reference)
   > AddChildToHorizontalBox(self=@n9, Content=Get Button_SizeBox_NotNow())
   n10 = SlotAsHorizontalBoxSlot!(Widget:Widget Object Reference=Get Button_SizeBox_Continue()) -> (ReturnValue:Horizontal Box Slot Object Reference)
   > SetHorizontalAlignment(self:Horizontal Box Slot Object Reference=@n10, InHorizontalAlignment=HAlign_Left)
   > SetPadding(self:Horizontal Box Slot Object Reference=@n10, InPadding_Left=0, InPadding_Top=0, InPadding_Right=16, InPadding_Bottom=0)
   n11 = SlotAsHorizontalBoxSlot(Widget=Get Button_SizeBox_NotNow())
   > SetHorizontalAlignment(self=@n11, InHorizontalAlignment=HAlign_Left)
   > SetPadding(self=@n11, InPadding_Left=0, InPadding_Top=0, InPadding_Right=0, InPadding_Bottom=0)
  ]}
 EventGraph {nodes [
   OnOrientationChanged!() -> (OutputDelegate, then, inScreenOrientation:EScreenOrientation Enum=Unknown) {is=title K2Node_CustomEvent}
   > AdaptToScreenOrientation() {is=title K2Node_CallFunction}
   Event Construct()
   > AdaptToScreenOrientation()
   > n1 = Cast To PlatformGameInstance!(execute, Object=GetGameInstance!(self:Widget Object Reference) -> (ReturnValue:Game Instance Object Reference)) -> (then, CastFailed, AsPlatform Game Instance:Platform Game Instance Object Reference) {is=title K2Node_DynamicCast}
   > n2 = Set GameInstance(GameInstance:Platform Game Instance Object Reference=@n1) -> (Output_Get:Platform Game Instance Object Reference)
   n3 = Create Event()
   > Bind Event to Application Received Screen Orientation Changed Notification Delegate(self:Platform Game Instance Object Reference=@n2, Delegate=@n3) {is=title K2Node_AddDelegate}
   Event Destruct()
   > Unbind Event from Application Received Screen Orientation Changed Notification Delegate(self:Platform Game Instance Object Reference=Get GameInstance() -> (GameInstance:Platform Game Instance Object Reference), Delegate=@n3) {is=title K2Node_RemoveDelegate}
  ]}
]
variables [GameInstance {type=PlatformGameInstance, default=null}]
widgets {
 info {inheritedWidgetCount=4, parentClass=/Script/FortniteUI.FortPushNotificationSelectScreen, rootWidgetClass=&Overlay, widgetCount=23}
 widgets [
  Overlay_0 {widgetClassPath=&Overlay, widget=Overlay_0}
  CommonBorder_85 {
   widgetClassPath=&CommonBorder
   parent=Overlay_0
   widget=CommonBorder_85
   properties {
    visibility=Visible
    background {
     bIsDynamicallyLoaded=false
     drawAs=Image
     imageType=NoImage
     mirroring=NoMirror
     resourceName=None
     resourceObject=null
     tiling=NoTile
     imageSize=<32 32>
     margin=<ltrb 0 0 0 0>
     outlineSettings {
      bUseBrushTransparency=false
      roundingType=HalfHeightRadius
      width=0
      color {colorUseRule=UseColor_Specified, specifiedColor=<rgba 0 0 0 0>}
      cornerRadii=<0 0 0 0>
     }
     tintColor {colorUseRule=UseColor_Specified, specifiedColor=<rgba 0 0 0 0.5>}
     uVRegion {bIsValid=false, max=<0 0>, min=<0 0>}
    }
   }
   slot {class=OverlaySlot, properties {horizontalAlignment=HAlign_Fill, verticalAlignment=VAlign_Fill}}
  }
  SafeZone {
   widgetClassPath=&SafeZone
   parent=Overlay_0
   widget=SafeZone
   properties {padLeft=false, padRight=false}
   slot {class=OverlaySlot, properties {horizontalAlignment=HAlign_Fill, verticalAlignment=VAlign_Fill}}
  }
  CommonBorder_0 {
   widgetClassPath=&CommonBorder
   parent=SafeZone
   widget=CommonBorder_0
   properties {
    background {
     bIsDynamicallyLoaded=false
     drawAs=RoundedBox
     imageType=FullColor
     mirroring=NoMirror
     resourceName=None
     resourceObject=null
     tiling=NoTile
     imageSize=<128 32>
     margin=<ltrb 0.1 0.45 0.9 0.5>
     outlineSettings {
      bUseBrushTransparency=false
      roundingType=FixedRadius
      width=0
      color {colorUseRule=UseColor_Specified, specifiedColor=<rgba 0 0 0 0>}
      cornerRadii=<20 20 20 20>
     }
     tintColor {colorUseRule=UseColor_Specified, specifiedColor=<rgba 0 0 0 0.95>}
     uVRegion {bIsValid=false, max=<0 0>, min=<0 0>}
    }
    brushColor=<rgba 0 0 0 0>
    padding=<ltrb 10 50 10 50>
   }
   slot {class=SafeZoneSlot, properties {hAlign=HAlign_Center, vAlign=VAlign_Center}}
  }
  SizeBox_45 {
   widgetClassPath=&SizeBox
   parent=CommonBorder_0
   widget=SizeBox_45
   properties {
    bOverride_MaxDesiredWidth=true
    bOverride_MinDesiredHeight=true
    bOverride_MinDesiredWidth=true
    maxDesiredWidth=1620
    minDesiredHeight=500
    minDesiredWidth=750
   }
   slot {class=BorderSlot, properties {padding=<ltrb 10 50 10 50>}}
  }
  Overlay_149 {widgetClassPath=&Overlay, parent=SizeBox_45, widget=Overlay_149, slot {class=SizeBoxSlot, properties {padding=<ltrb 30 2 30 2>}}}
  CommonBorder_87 {
   widgetClassPath=&CommonBorder
   parent=Overlay_149
   widget=CommonBorder_87
   properties {background {
     bIsDynamicallyLoaded=false
     drawAs=RoundedBox
     imageType=NoImage
     mirroring=NoMirror
     resourceName=None
     resourceObject=null
     tiling=NoTile
     imageSize=<32 32>
     margin=<ltrb 0 0 0 0>
     outlineSettings {
      bUseBrushTransparency=false
      roundingType=FixedRadius
      width=0
      color {colorUseRule=UseColor_Specified, specifiedColor=<rgba 0 0 0 0>}
      cornerRadii=<15 15 15 15>
     }
     tintColor {colorUseRule=UseColor_Specified, specifiedColor=<rgba 0.0091 0.0144 0.0262 1>}
     uVRegion {bIsValid=false, max=<0 0>, min=<0 0>}
    }}
   slot {class=OverlaySlot, properties {horizontalAlignment=HAlign_Fill, verticalAlignment=VAlign_Fill}}
  }
  VerticalBox_0 {
   widgetClassPath=&VerticalBox
   parent=Overlay_149
   widget=VerticalBox_0
   slot {class=OverlaySlot, properties {horizontalAlignment=HAlign_Fill, verticalAlignment=VAlign_Fill}}
  }
  VerticalBox_1 {
   widgetClassPath=&VerticalBox
   parent=VerticalBox_0
   widget=VerticalBox_1
   slot {class=VerticalBoxSlot, properties {size {sizeRule=Fill, value=1}}}
  }
  SizeBox_47 {
   widgetClassPath=&SizeBox
   parent=VerticalBox_1
   widget=SizeBox_47
   properties {bOverride_MinDesiredHeight=true, minDesiredHeight=100}
   slot {class=VerticalBoxSlot}
  }
  CommonBorder_1 {
   widgetClassPath=&CommonBorder
   parent=SizeBox_47
   widget=CommonBorder_1
   properties {
    verticalAlignment=VAlign_Center
    background {
     bIsDynamicallyLoaded=false
     drawAs=RoundedBox
     imageType=NoImage
     mirroring=NoMirror
     resourceName=None
     resourceObject=null
     tiling=NoTile
     imageSize=<32 32>
     margin=<ltrb 0 0 0 0>
     outlineSettings {
      bUseBrushTransparency=false
      roundingType=FixedRadius
      width=0
      color {colorUseRule=UseColor_Specified, specifiedColor=<rgba 0 0 0 0>}
      cornerRadii=<15 15 0 0>
     }
     tintColor {colorUseRule=UseColor_Specified, specifiedColor=<rgba 0.0048 0.007 0.0103 1>}
     uVRegion {bIsValid=false, max=<0 0>, min=<0 0>}
    }
    padding=<ltrb 75 40 75 40>
   }
   slot {class=SizeBoxSlot}
  }
  Title {
   widgetClassPath=&CommonTextBlock
   bInherited=true
   parent=CommonBorder_1
   widget=Title
   properties {
    autoWrapText=true
    bIsScrollingEnabled=false
    justification=Center
    text=GET NOTIFIED
    font {
     bForceMonospaced=false
     bMaterialIsStencil=false
     fontMaterial=null
     fontObject=&HeadingNow
     letterSpacing=-40
     monospacedWidth=1
     size=36
     skewAmount=0
     typefaceFontName=76 Bold
     outlineSettings {
      bApplyOutlineToDropShadows=false
      bMiteredCorners=false
      bSeparateFillAlpha=false
      outlineMaterial=null
      outlineSize=0
      outlineColor=<rgba 0 0 0 1>
     }
    }
    shadowOffset=<0 0>
   }
   slot {class=BorderSlot, properties {verticalAlignment=VAlign_Center, padding=<ltrb 75 40 75 40>}}
  }
  VerticalBox_2 {
   widgetClassPath=&VerticalBox
   parent=VerticalBox_1
   widget=VerticalBox_2
   slot {class=VerticalBoxSlot, properties {size {sizeRule=Fill, value=1}}}
  }
  ScrollBox_0 {
   widgetClassPath=&ScrollBox
   parent=VerticalBox_2
   widget=ScrollBox_0
   properties {widgetBarStyle {
     thickness=16
     draggedThumbImage {
      bIsDynamicallyLoaded=false
      drawAs=Box
      imageType=FullColor
      mirroring=NoMirror
      resourceName=../../../Engine/Content/Slate/Common/Scrollbar_Thumb.png
      resourceObject=null
      tiling=NoTile
      imageSize=<0 0>
      margin=<ltrb 0.25 0.25 0.25 0.25>
      outlineSettings {
       bUseBrushTransparency=false
       roundingType=HalfHeightRadius
       width=0
       color {colorUseRule=UseColor_Specified, specifiedColor=<rgba 0 0 0 0>}
       cornerRadii=<0 0 0 0>
      }
      tintColor {colorUseRule=UseColor_Specified, specifiedColor=<rgba 1 1 1 1>}
      uVRegion {bIsValid=false, max=<0 0>, min=<0 0>}
     }
     horizontalBackgroundImage {
      bIsDynamicallyLoaded=false
      drawAs=NoDrawType
      imageType=NoImage
      mirroring=NoMirror
      resourceName=None
      resourceObject=null
      tiling=NoTile
      imageSize=<0 0>
      margin=<ltrb 0 0 0 0>
      outlineSettings {
       bUseBrushTransparency=false
       roundingType=HalfHeightRadius
       width=0
       color {colorUseRule=UseColor_Specified, specifiedColor=<rgba 0 0 0 0>}
       cornerRadii=<0 0 0 0>
      }
      tintColor {colorUseRule=UseColor_Specified, specifiedColor=<rgba 1 1 1 1>}
      uVRegion {bIsValid=false, max=<0 0>, min=<0 0>}
     }
     horizontalBottomSlotImage {
      bIsDynamicallyLoaded=false
      drawAs=Image
      imageType=FullColor
      mirroring=NoMirror
      resourceName=../../../Engine/Content/Slate/Common/Scrollbar_Background_Horizontal.png
      resourceObject=null
      tiling=NoTile
      imageSize=<8 8>
      margin=<ltrb 0 0 0 0>
      outlineSettings {
       bUseBrushTransparency=false
       roundingType=HalfHeightRadius
       width=0
       color {colorUseRule=UseColor_Specified, specifiedColor=<rgba 0 0 0 0>}
       cornerRadii=<0 0 0 0>
      }
      tintColor {colorUseRule=UseColor_Specified, specifiedColor=<rgba 1 1 1 1>}
      uVRegion {bIsValid=false, max=<0 0>, min=<0 0>}
     }
     horizontalTopSlotImage {
      bIsDynamicallyLoaded=false
      drawAs=Image
      imageType=FullColor
      mirroring=NoMirror
      resourceName=../../../Engine/Content/Slate/Common/Scrollbar_Background_Horizontal.png
      resourceObject=null
      tiling=NoTile
      imageSize=<8 8>
      margin=<ltrb 0 0 0 0>
      outlineSettings {
       bUseBrushTransparency=false
       roundingType=HalfHeightRadius
       width=0
       color {colorUseRule=UseColor_Specified, specifiedColor=<rgba 0 0 0 0>}
       cornerRadii=<0 0 0 0>
      }
      tintColor {colorUseRule=UseColor_Specified, specifiedColor=<rgba 1 1 1 1>}
      uVRegion {bIsValid=false, max=<0 0>, min=<0 0>}
     }
     hoveredThumbImage {
      bIsDynamicallyLoaded=false
      drawAs=Box
      imageType=FullColor
      mirroring=NoMirror
      resourceName=../../../Engine/Content/Slate/Common/Scrollbar_Thumb.png
      resourceObject=null
      tiling=NoTile
      imageSize=<0 0>
      margin=<ltrb 0.25 0.25 0.25 0.25>
      outlineSettings {
       bUseBrushTransparency=false
       roundingType=HalfHeightRadius
       width=0
       color {colorUseRule=UseColor_Specified, specifiedColor=<rgba 0 0 0 0>}
       cornerRadii=<0 0 0 0>
      }
      tintColor {colorUseRule=UseColor_Specified, specifiedColor=<rgba 1 1 1 1>}
      uVRegion {bIsValid=false, max=<0 0>, min=<0 0>}
     }
     normalThumbImage {
      bIsDynamicallyLoaded=false
      drawAs=Box
      imageType=FullColor
      mirroring=NoMirror
      resourceName=../../../Engine/Content/Slate/Common/Scrollbar_Thumb.png
      resourceObject=null
      tiling=NoTile
      imageSize=<0 0>
      margin=<ltrb 0.25 0.25 0.25 0.25>
      outlineSettings {
       bUseBrushTransparency=false
       roundingType=HalfHeightRadius
       width=0
       color {colorUseRule=UseColor_Specified, specifiedColor=<rgba 0 0 0 0>}
       cornerRadii=<0 0 0 0>
      }
      tintColor {colorUseRule=UseColor_Specified, specifiedColor=<rgba 1 1 1 1>}
      uVRegion {bIsValid=false, max=<0 0>, min=<0 0>}
     }
     verticalBackgroundImage {
      bIsDynamicallyLoaded=false
      drawAs=Image
      imageType=NoImage
      mirroring=NoMirror
      resourceName=None
      resourceObject=null
      tiling=NoTile
      imageSize=<0 0>
      margin=<ltrb 0 0 0 0>
      outlineSettings {
       bUseBrushTransparency=false
       roundingType=HalfHeightRadius
       width=0
       color {colorUseRule=UseColor_Specified, specifiedColor=<rgba 0 0 0 0>}
       cornerRadii=<0 0 0 0>
      }
      tintColor {colorUseRule=UseColor_Specified, specifiedColor=<rgba 1 1 1 0.5>}
      uVRegion {bIsValid=false, max=<0 0>, min=<0 0>}
     }
     verticalBottomSlotImage {
      bIsDynamicallyLoaded=false
      drawAs=Image
      imageType=FullColor
      mirroring=NoMirror
      resourceName=../../../Engine/Content/Slate/Common/Scrollbar_Background_Vertical.png
      resourceObject=null
      tiling=NoTile
      imageSize=<8 8>
      margin=<ltrb 0 0 0 0>
      outlineSettings {
       bUseBrushTransparency=false
       roundingType=HalfHeightRadius
       width=0
       color {colorUseRule=UseColor_Specified, specifiedColor=<rgba 0 0 0 0>}
       cornerRadii=<0 0 0 0>
      }
      tintColor {colorUseRule=UseColor_Specified, specifiedColor=<rgba 1 1 1 1>}
      uVRegion {bIsValid=false, max=<0 0>, min=<0 0>}
     }
     verticalTopSlotImage {
      bIsDynamicallyLoaded=false
      drawAs=Image
      imageType=FullColor
      mirroring=NoMirror
      resourceName=../../../Engine/Content/Slate/Common/Scrollbar_Background_Vertical.png
      resourceObject=null
      tiling=NoTile
      imageSize=<8 8>
      margin=<ltrb 0 0 0 0>
      outlineSettings {
       bUseBrushTransparency=false
       roundingType=HalfHeightRadius
       width=0
       color {colorUseRule=UseColor_Specified, specifiedColor=<rgba 0 0 0 0>}
       cornerRadii=<0 0 0 0>
      }
      tintColor {colorUseRule=UseColor_Specified, specifiedColor=<rgba 1 1 1 1>}
      uVRegion {bIsValid=false, max=<0 0>, min=<0 0>}
     }
    }}
   slot {class=VerticalBoxSlot, properties {padding=<ltrb 75 60 40 20>, size {sizeRule=Fill, value=1}}}
  }
  Description {
   widgetClassPath=&CommonTextBlock
   bInherited=true
   parent=ScrollBox_0
   widget=Description
   properties {
    autoWrapText=true
    lineHeightPercentage=1.2
    text="Get notified about friend and party invites, new seasons, limited time events, announcements and promotions. Configure these at any time in Fortnite Settings."
    font {
     bForceMonospaced=false
     bMaterialIsStencil=false
     fontMaterial=null
     fontObject=&HeadingNow
     letterSpacing=0
     monospacedWidth=1
     size=31.5
     skewAmount=0
     typefaceFontName=74 Regular
     outlineSettings {
      bApplyOutlineToDropShadows=false
      bMiteredCorners=false
      bSeparateFillAlpha=false
      outlineMaterial=null
      outlineSize=0
      outlineColor=<rgba 0 0 0 1>
     }
    }
    shadowOffset=<0 0>
   }
   slot {class=ScrollBoxSlot, properties {padding=<ltrb 0 0 35 0>}}
  }
  HorizontalBox_Buttons {
   widgetClassPath=&HorizontalBox
   bIsVariable=true
   parent=VerticalBox_0
   widget=HorizontalBox_Buttons
   slot {class=VerticalBoxSlot, properties {horizontalAlignment=HAlign_Left, verticalAlignment=VAlign_Bottom, padding=<ltrb 75 16 75 50>}}
  }
  VerticalBox_Buttons {widgetClassPath=&VerticalBox, bIsVariable=true, parent=VerticalBox_0, widget=VerticalBox_Buttons, slot {class=VerticalBoxSlot}}
  Button_SizeBox_NotNow {
   widgetClassPath=&SizeBox
   bIsVariable=true
   parent=VerticalBox_Buttons
   widget=Button_SizeBox_NotNow
   slot {class=VerticalBoxSlot}
  }
  Overlay_Button_NotNow {widgetClassPath=&Overlay, bIsVariable=true, parent=Button_SizeBox_NotNow, widget=Overlay_Button_NotNow, slot {class=SizeBoxSlot}}
  Button_NotNow {
   widgetClassPath=&WBP_UIKit_Button_Regular
   bInherited=true
   bIsVariable=true
   parent=Overlay_Button_NotNow
   widget=Button_NotNow
   properties {buttonSize=Large, text=Not Now, triggeringInputAction {dataTable=&InputActionDataTable, rowName=Decline}}
   slot {class=OverlaySlot, properties {horizontalAlignment=HAlign_Fill, verticalAlignment=VAlign_Fill}}
  }
  Button_SizeBox_Continue {
   widgetClassPath=&SizeBox
   bIsVariable=true
   parent=VerticalBox_Buttons
   widget=Button_SizeBox_Continue
   slot {class=VerticalBoxSlot}
  }
  Overlay_Button_Continue {
   widgetClassPath=&Overlay
   bIsVariable=true
   parent=Button_SizeBox_Continue
   widget=Overlay_Button_Continue
   slot {class=SizeBoxSlot}
  }
  Button_Continue {
   widgetClassPath=/FortUILibrary/UIKit/Buttons/WBP_UIKit_Button_Loud
   bInherited=true
   bIsVariable=true
   parent=Overlay_Button_Continue
   widget=Button_Continue
   properties {buttonSize=Large, text=Continue, triggeringInputAction {dataTable=&InputActionDataTable, rowName=Continue}}
   slot {class=OverlaySlot, properties {horizontalAlignment=HAlign_Fill, verticalAlignment=VAlign_Fill}}
  }
 ]
}
