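# Fortnite splash screen widget, built on FortSplashScreenWidget, showing the wordmark logo above a start button bound to the Confirm action.
# On initialization it hides the logo if ShouldHideLogo is true, and button clicks fire an element-interacted event.
# Input method changes choose "Press to Start" for gamepad or touch platforms, otherwise "Click to Start".
# Activation sets "Touch to Start" on mobile apps and focuses the button after a short delay.
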WidgetBlueprint /FortniteCoreUI/UI/Login/SplashScreen/SplashScreenWidget
graphs [EventGraph {nodes [
   n1 = "On Clicked (Button_PressToStart)"!() -> (OutputDelegate, then, Button:Common Button Base Object Reference) {is=title K2Node_ComponentBoundEvent}
   > Fire Event Element Interacted(Element=@n1.Button)
   Event OnInitialized()
   > SetVisibility(
    self=Get I_Fortnite() -> (I_Fortnite:Common Lazy Image Object Reference)
    InVisibility=Select(Option 0=HitTestInvisible, Option 1=Hidden, Index=ShouldHideLogo!(self:Fort Splash Screen Widget Object Reference) -> (ReturnValue:bool))
   )
   n2 = Event OnInputMethodChanged!() -> (OutputDelegate, then, InputType:ECommonInputType Enum) {is=title K2Node_Event}
   n3 = IsTouchInputPlatform!() -> (ReturnValue:bool)
   n4 = "Equal (Enum)"(A=@n2.InputType, B=Gamepad)
   > SetText!(
    execute
    self:Fort CTAButton Object Reference=Get Button_PressToStart() -> (Button_PressToStart:WBP UIKit Button Loud Object Reference)
    ButtonText:"Text (by ref)"=Select(Option 0:Text=Get MouseText() -> (MouseText:Text), Option 1:Text=Get GamepadText() -> (GamepadText:Text), Index=OR Boolean(A=@n4, B=@n3)) -> (ReturnValue:Text)
   ) -> (then)
   Event On Activated()
   > n5 = Sequence()
   Delay(execute=@n5.then_1, Duration=0.5)
   n6 = Get Button_PressToStart()
   > SetFocus(self=@n6)
   > EventOnFocusedManually(self:WBP UIKit Button CTA Base Object Reference=@n6)
   Branch(
    execute=@n5.then_0
    Condition=IsMobileApp!(self:Fort Global UIContext Object Reference=Get FortGlobalUIContext!() -> (ReturnValue:Fort Global UIContext Object Reference) {is=title K2Node_GetBlueprintContextEx}) -> (ReturnValue:bool)
   )
   > SetText(self=Get Button_PressToStart(), ButtonText=Get MobileText() -> (MobileText:Text))
   Event On Deactivated()
  ]}]
variables [
 GamepadText {type=string, default=Press to Start}
 MobileText {type=string, default=Touch to Start}
 MouseText {type=string, default=Click to Start}
]
widgets {
 info {inheritedWidgetCount=0, parentClass=/Script/FortniteUI.FortSplashScreenWidget, rootWidgetClass=&SizeBox, widgetCount=4}
 widgets [
  SizeBox_0 {
   widgetClassPath=&SizeBox
   widget=SizeBox_0
   properties {bOverride_HeightOverride=true, bOverride_MinDesiredWidth=true, heightOverride=640, minDesiredWidth=1300, visibility=HitTestInvisible}
  }
  VerticalBox_0 {
   widgetClassPath=&VerticalBox
   parent=SizeBox_0
   widget=VerticalBox_0
   slot {class=SizeBoxSlot, properties {verticalAlignment=VAlign_Bottom}}
  }
  I_Fortnite {
   widgetClassPath=/Script/CommonUI.CommonLazyImage
   bIsVariable=true
   parent=VerticalBox_0
   widget=I_Fortnite
   properties {
    visibility=HitTestInvisible
    brush {
     bIsDynamicallyLoaded=false
     drawAs=Image
     imageType=NoImage
     mirroring=NoMirror
     resourceName=None
     resourceObject=/FortniteCoreUI/Foundation/Textures/Logo/T-Logo-Wordmark-1024x256
     tiling=NoTile
     imageSize=<1024 256>
     margin=<ltrb 0 0 0 0>
     outlineSettings {
      bUseBrushTransparency=false
      roundingType=HalfHeightRadius
      width=0
      color {colorUseRule=UseColor_Specified, specifiedColor=<rgba 0 0 0 0>}
      cornerRadii=<0 0 0 0>
     }
     tintColor {colorUseRule=UseColor_Specified, specifiedColor=<rgba 1 1 1 1>}
     uVRegion {bIsValid=false, max=<0 0>, min=<0 0>}
    }
    renderTransform {angle=0, scale=<0.65 0.65>, shear=<0 0>, translation=<0 45>}
   }
   slot {class=VerticalBoxSlot, properties {horizontalAlignment=HAlign_Center, verticalAlignment=VAlign_Center}}
  }
  Button_PressToStart {
   widgetClassPath=/FortUILibrary/UIKit/Buttons/WBP_UIKit_Button_Loud
   bIsVariable=true
   parent=VerticalBox_0
   widget=Button_PressToStart
   properties {buttonSize=Large, text=PRESS TO START, triggeringInputAction {dataTable=&InputActionDataTable, rowName=Confirm}}
   slot {class=VerticalBoxSlot, properties {horizontalAlignment=HAlign_Center, verticalAlignment=VAlign_Bottom, padding=<ltrb 0 64 0 64>}}
  }
 ]
}
