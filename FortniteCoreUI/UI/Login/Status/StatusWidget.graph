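# Login status widget derived from FortLoginStatus, showing a title, throbber, status text and Continue/Retry, Quit and online-status buttons.
# On Construct, ShouldUseBootUpRedesign picks the legacy or redesigned button row in a switcher.
# The redesign also restyles the title, hides the throbber and re-adds content children to reorder the layout.
# Clicks fire interaction events; continue text follows the active design, while quit text only updates the legacy button.
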WidgetBlueprint /FortniteCoreUI/UI/Login/Status/StatusWidget
graphs [
 EventGraph {nodes [
   n1 = "On Button Clicked (Button_Continue)"!() -> (OutputDelegate, then, Button:Common Button Legacy Object Reference) {is=title K2Node_ComponentBoundEvent}
   > Fire Event Element Interacted(Element=@n1.Button)
   n2 = "On Button Clicked (Button_Quit)"!() -> (OutputDelegate, then, Button:Common Button Legacy Object Reference) {is=title K2Node_ComponentBoundEvent}
   > Fire Event Element Interacted(Element=@n2.Button)
   n3 = "On Button Clicked (Button_Status)"!() -> (OutputDelegate, then, Button:Common Button Legacy Object Reference) {is=title K2Node_ComponentBoundEvent}
   > Fire Event Element Interacted(Element=@n3.Button)
   n4 = "On Clicked (Button_Continue_Redesign)"!() -> (OutputDelegate, then, Button:Common Button Base Object Reference) {is=title K2Node_ComponentBoundEvent}
   > Fire Event Element Interacted(Element=@n4.Button)
   n5 = "On Clicked (Button_Quit_Redesign)"!() -> (OutputDelegate, then, Button:Common Button Base Object Reference) {is=title K2Node_ComponentBoundEvent}
   > Fire Event Element Interacted(Element=@n5.Button)
   n6 = Event SetContinueButtonText!() -> (OutputDelegate, then, ContinueText:"Text (by ref)") {is=title K2Node_Event}
   > n7 = Branch(Condition=ShouldUseBootUpRedesign!(self:Fort Login Status Object Reference) -> (ReturnValue:bool))
   > n8 = Cast To WBP_UIKit_ButtonCTA_Base!(execute, Object=Get Button_Continue_Redesign() -> (Button_Continue_Redesign:Common Button Base Object Reference)) -> (then, CastFailed, AsWBP UIKit Button CTA Base:WBP UIKit Button CTA Base Object Reference) {is=title K2Node_DynamicCast}
   > SetText!(execute, self:Fort CTAButton Object Reference=@n8, ButtonText:"Text (by ref)"=@n6.ContinueText) -> (then)
   n9 = Cast To IconTextButton!(execute=@n7.else, Object=Get Button_Continue() -> (Button_Continue:Common Button Legacy Object Reference)) -> (then, CastFailed, AsIcon Text Button:Icon Text Button Object Reference) {is=title K2Node_DynamicCast}
   > Set Text!(execute, self:Icon Text Button Object Reference=@n9, ButtonText:Text=@n6.ContinueText) -> (then) {is=title K2Node_CallFunction}
   n10 = Event SetQuitButtonText!() -> (OutputDelegate, then, QuitText:"Text (by ref)") {is=title K2Node_Event}
   > n11 = Cast To IconTextButton(Object=Get Button_Quit() -> (Button_Quit:Common Button Legacy Object Reference))
   > Set Text(self=@n11, ButtonText=@n10.QuitText)
   Event Construct()
   > n12 = Branch(Condition=ShouldUseBootUpRedesign())
   > SetRedesignedStyle()
   > SetActiveWidget(
    self:Widget Switcher Object Reference=Get Switcher_Button_Default_Redesign() -> (Switcher_Button_Default_Redesign:Common Activatable Widget Switcher Object Reference)
    Widget=Get HorizontalBox_Redesign() -> (HorizontalBox_Redesign:Horizontal Box Object Reference)
   )
   SetActiveWidget(
    execute=@n12.else
    self=Get Switcher_Button_Default_Redesign()
    Widget=Get HorizontalBox_Default() -> (HorizontalBox_Default:Horizontal Box Object Reference)
   )
  ]}
 SetRedesignedStyle {nodes [
   SetRedesignedStyle() {is=title K2Node_FunctionEntry}
   n1 = Get Text_Title() -> (Text_Title:Common Text Object Reference)
   > SetRenderShear!(execute, self:Widget Object Reference=@n1, Shear:Vector 2D Structure) -> (then)
   > SetTextTransformPolicy!(execute, self:Text Object Reference=@n1, InTransformPolicy:ETextTransformPolicy Enum=None) -> (then)
   > SetShadowOffset!(execute, self:Text Object Reference=@n1, InShadowOffset:Vector 2D Structure) -> (then)
   > SetColorAndOpacity!(
    execute
    self=@n1
    InColorAndOpacity_SpecifiedColor:Linear Color Structure=(R=1,G=1,B=1,A=1)
    InColorAndOpacity_ColorUseRule:ESlateColorStylingMode Enum
   ) -> (then)
   > SetFont!(
    execute
    self=@n1
    InFontInfo_FontObject:Object Reference=&HeadingNow
    InFontInfo_FontMaterial:Object Reference
    InFontInfo_OutlineSettings:Font Outline Settings Structure
    InFontInfo_TypefaceFontName:Name=74 Regular
    InFontInfo_Size:float=33
    InFontInfo_LetterSpacing:int=0
    InFontInfo_SkewAmount:float=0
    InFontInfo_bForceMonospaced:bool=false
    InFontInfo_bMaterialIsStencil:bool=false
    InFontInfo_MonospacedWidth:float=1
   ) -> (then)
   > SetVisibility(self=Get Image_Throbber() -> (Image_Throbber:Image Object Reference), InVisibility=Collapsed)
   n2 = Get VerticalBox_Content() -> (VerticalBox_Content:Vertical Box Object Reference)
   > n3 = AddChildToVerticalBox(self=@n2, Content=Get VerticalBox_Title() -> (VerticalBox_Title:Vertical Box Object Reference))
   > SetHorizontalAlignment(self:Vertical Box Slot Object Reference=@n3, InHorizontalAlignment=HAlign_Center)
   > SetVerticalAlignment(self:Vertical Box Slot Object Reference=@n3, InVerticalAlignment=VAlign_Top)
   > SetPadding!(
    execute
    self:Vertical Box Slot Object Reference=@n3
    InPadding_Left:float=0
    InPadding_Top:float=10
    InPadding_Right:float=0
    InPadding_Bottom:float=20
   ) -> (then)
   > n4 = AddChildToVerticalBox(self=@n2, Content=Get Switcher_Button_Default_Redesign())
   > SetHorizontalAlignment(self=@n4, InHorizontalAlignment=HAlign_Center)
   > SetVerticalAlignment(self=@n4, InVerticalAlignment=VAlign_Bottom)
   > SetPadding(self=@n4, InPadding_Left=0, InPadding_Top=10, InPadding_Right=0, InPadding_Bottom=20)
   > n5 = AddChildToVerticalBox(self=@n2, Content=Get SizeBox_Text_Status() -> (SizeBox_Text_Status:Size Box Object Reference))
   > SetHorizontalAlignment(self=@n5, InHorizontalAlignment=HAlign_Fill)
   > SetVerticalAlignment(self=@n5, InVerticalAlignment=VAlign_Top)
   > SetPadding(
    self:Overlay Slot Object Reference=SlotAsOverlaySlot(Widget=Get VerticalBox_Content())
    InPadding_Left=0
    InPadding_Top=0
    InPadding_Right=0
    InPadding_Bottom=40
   )
   > SetSidesToPad!(
    execute
    self:Safe Zone Object Reference=Get SafeZone_Title() -> (SafeZone_Title:Safe Zone Object Reference)
    InPadLeft:bool=false
    InPadRight:bool=false
    InPadTop:bool=false
    InPadBottom:bool=true
   ) -> (then)
  ]}
]
variables []
widgets {
 info {
  inheritedWidgetCount=7
  namedSlotCount=1
  parentClass=/Script/FortniteUI.FortLoginStatus
  rootWidgetClass=&SizeBox
  widgetCount=24
 }
 widgets [
  SizeBox_1 {widgetClassPath=&SizeBox, widget=SizeBox_1, properties {bOverride_HeightOverride=true, heightOverride=640}}
  Overlay_0 {widgetClassPath=&Overlay, parent=SizeBox_1, widget=Overlay_0, slot {class=SizeBoxSlot}}
  VerticalBox_Content {
   widgetClassPath=&VerticalBox
   bIsVariable=true
   parent=Overlay_0
   widget=VerticalBox_Content
   slot {class=OverlaySlot, properties {horizontalAlignment=HAlign_Fill, verticalAlignment=VAlign_Bottom}}
  }
  CommonLazyImage_0 {
   widgetClassPath=/Script/CommonUI.CommonLazyImage
   parent=VerticalBox_Content
   widget=CommonLazyImage_0
   properties {
    visibility=Collapsed
    brush {
     bIsDynamicallyLoaded=false
     drawAs=Image
     imageType=NoImage
     mirroring=NoMirror
     resourceName=None
     resourceObject=/FortniteCoreUI/Foundation/Textures/Logo/T-Logo-Wordmark-1024x256
     tiling=NoTile
     imageSize=<1024 256>
     margin=<ltrb 0 0 0 0>
     outlineSettings {
      bUseBrushTransparency=false
      roundingType=HalfHeightRadius
      width=0
      color {colorUseRule=UseColor_Specified, specifiedColor=<rgba 0 0 0 0>}
      cornerRadii=<0 0 0 0>
     }
     tintColor {colorUseRule=UseColor_Specified, specifiedColor=<rgba 1 1 1 1>}
     uVRegion {bIsValid=false, max=<0 0>, min=<0 0>}
    }
   }
   slot {class=VerticalBoxSlot, properties {horizontalAlignment=HAlign_Center, verticalAlignment=VAlign_Top}}
  }
  VerticalBox_Title {
   widgetClassPath=&VerticalBox
   bIsVariable=true
   parent=VerticalBox_Content
   widget=VerticalBox_Title
   slot {class=VerticalBoxSlot, properties {horizontalAlignment=HAlign_Center, verticalAlignment=VAlign_Top, padding=<ltrb 0 10 0 20>}}
  }
  Border_3 {
   widgetClassPath=&Border
   parent=VerticalBox_Title
   widget=Border_3
   properties {
    background {
     bIsDynamicallyLoaded=false
     drawAs=NoDrawType
     imageType=NoImage
     mirroring=NoMirror
     resourceName=None
     resourceObject=null
     tiling=NoTile
     imageSize=<32 32>
     margin=<ltrb 0 0 0 0>
     outlineSettings {
      bUseBrushTransparency=false
      roundingType=HalfHeightRadius
      width=0
      color {colorUseRule=UseColor_Specified, specifiedColor=<rgba 0 0 0 0>}
      cornerRadii=<0 0 0 0>
     }
     tintColor {colorUseRule=UseColor_Specified, specifiedColor=<rgba 1 1 1 1>}
     uVRegion {bIsValid=false, max=<0 0>, min=<0 0>}
    }
    padding=<ltrb 0 0 0 0>
   }
   slot {class=VerticalBoxSlot}
  }
  VerticalBox_2 {widgetClassPath=&VerticalBox, parent=Border_3, widget=VerticalBox_2, slot {class=BorderSlot, properties {padding=<ltrb 0 0 0 0>}}}
  SafeZone_Title {
   widgetClassPath=&SafeZone
   bIsVariable=true
   parent=VerticalBox_2
   widget=SafeZone_Title
   properties {padBottom=false, padLeft=false, padRight=false, padTop=false}
   slot {class=VerticalBoxSlot, properties {verticalAlignment=VAlign_Top}}
  }
  Text_Title {
   widgetClassPath=&CommonTextBlock
   bInherited=true
   bIsVariable=true
   parent=SafeZone_Title
   widget=Text_Title
   properties {
    text=Logging In...
    textTransformPolicy=ToUpper
    wrapTextAt=1200
    colorAndOpacity {colorUseRule=UseColor_Specified, specifiedColor=<rgba 0.0331 0.855 1 1>}
    font {
     bForceMonospaced=false
     bMaterialIsStencil=false
     fontMaterial=null
     fontObject=&BurbankSmall
     letterSpacing=0
     monospacedWidth=1
     size=32
     skewAmount=0
     typefaceFontName=Black
     outlineSettings {
      bApplyOutlineToDropShadows=false
      bMiteredCorners=false
      bSeparateFillAlpha=false
      outlineMaterial=null
      outlineSize=0
      outlineColor=<rgba 0 0 0 1>
     }
    }
    renderTransform {angle=0, scale=<1 1>, shear=<-12 0>, translation=<0 0>}
    shadowColorAndOpacity=<rgba 0 0.05 0.15 0.8>
    shadowOffset=<3 3>
   }
   slot {class=SafeZoneSlot}
  }
  Image_Throbber {
   widgetClassPath=&Image
   bIsVariable=true
   parent=VerticalBox_2
   widget=Image_Throbber
   properties {brush {
     bIsDynamicallyLoaded=false
     drawAs=Image
     imageType=NoImage
     mirroring=NoMirror
     resourceName=None
     resourceObject=/FortUILibrary/UIKit/Throbber/MI_UIKit_Monochrome_Throbber
     tiling=NoTile
     imageSize=<32 32>
     margin=<ltrb 0 0 0 0>
     outlineSettings {
      bUseBrushTransparency=false
      roundingType=HalfHeightRadius
      width=0
      color {colorUseRule=UseColor_Specified, specifiedColor=<rgba 0 0 0 0>}
      cornerRadii=<0 0 0 0>
     }
     tintColor {colorUseRule=UseColor_Specified, specifiedColor=<rgba 1 1 1 1>}
     uVRegion {bIsValid=false, max=<0 0>, min=<0 0>}
    }}
   slot {class=VerticalBoxSlot, properties {horizontalAlignment=HAlign_Center, verticalAlignment=VAlign_Bottom, padding=<ltrb 0 28 0 0>}}
  }
  Border_1 {
   widgetClassPath=&Border
   parent=VerticalBox_Title
   widget=Border_1
   properties {
    horizontalAlignment=HAlign_Left
    verticalAlignment=VAlign_Center
    background {
     bIsDynamicallyLoaded=false
     drawAs=NoDrawType
     imageType=NoImage
     mirroring=NoMirror
     resourceName=None
     resourceObject=null
     tiling=NoTile
     imageSize=<32 32>
     margin=<ltrb 0 0 0 0>
     outlineSettings {
      bUseBrushTransparency=false
      roundingType=HalfHeightRadius
      width=0
      color {colorUseRule=UseColor_Specified, specifiedColor=<rgba 0 0 0 0>}
      cornerRadii=<0 0 0 0>
     }
     tintColor {colorUseRule=UseColor_Specified, specifiedColor=<rgba 1 1 1 1>}
     uVRegion {bIsValid=false, max=<0 0>, min=<0 0>}
    }
    padding=<ltrb 4 -4 0 0>
   }
   slot {class=VerticalBoxSlot}
  }
  Image_4 {
   widgetClassPath=&Image
   parent=Border_1
   widget=Image_4
   properties {brush {
     bIsDynamicallyLoaded=false
     drawAs=Image
     imageType=NoImage
     mirroring=NoMirror
     resourceName=None
     resourceObject=/FortniteCoreUI/UI/Login/Art/M-Icon-ExternalLink-64
     tiling=NoTile
     imageSize=<70 70>
     margin=<ltrb 0 0 0 0>
     outlineSettings {
      bUseBrushTransparency=false
      roundingType=HalfHeightRadius
      width=0
      color {colorUseRule=UseColor_Specified, specifiedColor=<rgba 0 0 0 0>}
      cornerRadii=<0 0 0 0>
     }
     tintColor {colorUseRule=UseColor_Specified, specifiedColor=<rgba 1 1 1 1>}
     uVRegion {bIsValid=false, max=<0 0>, min=<0 0>}
    }}
   slot {class=BorderSlot, properties {horizontalAlignment=HAlign_Left, verticalAlignment=VAlign_Center, padding=<ltrb 4 -4 0 0>}}
  }
  Switcher_Button_Default_Redesign {
   widgetClassPath=&CommonActivatableWidgetSwitcher
   bIsVariable=true
   parent=VerticalBox_Content
   widget=Switcher_Button_Default_Redesign
   slot {class=VerticalBoxSlot, properties {horizontalAlignment=HAlign_Center, verticalAlignment=VAlign_Bottom, padding=<ltrb 0 20 0 10>}}
  }
  HorizontalBox_Default {
   widgetClassPath=&HorizontalBox
   bIsVariable=true
   parent=Switcher_Button_Default_Redesign
   widget=HorizontalBox_Default
   slot {class=WidgetSwitcherSlot}
  }
  Button_Continue {
   widgetClassPath=&IconTextButton
   bInherited=true
   bIsVariable=true
   parent=HorizontalBox_Default
   widget=Button_Continue
   properties {
    bIgnoreInputActionWidgetText=true
    buttonText=Retry
    onButtonBaseClicked=(&self:WidgetTree.Button_Continue.HandleOnButtonClicked)
    onButtonBaseDoubleClicked=(&self:WidgetTree.Button_Continue.HandleOnButtonDoubleClicked)
    onButtonBaseHovered=(&self:WidgetTree.Button_Continue.HandleOnButtonHovered)
    onButtonBaseUnhovered=(&self:WidgetTree.Button_Continue.HandleOnButtonUnhovered)
    onSelectedChangedBase=(&self:WidgetTree.Button_Continue.HandleOnSelectedChanged)
    style=/FortUILibrary/Foundation/Buttons/ButtonStyles/ButtonStyle_Feature_M_Yellow.ButtonStyle_Feature_M_Yellow_C
    visibility=Collapsed
    triggeringInputAction {dataTable=&InputActionDataTable, rowName=RetryLogin}
   }
   slot {class=HorizontalBoxSlot, properties {horizontalAlignment=HAlign_Center}}
  }
  Button_Quit {
   widgetClassPath=&IconTextButton
   bInherited=true
   bIsVariable=true
   parent=HorizontalBox_Default
   widget=Button_Quit
   properties {
    buttonText=Quit
    onButtonBaseClicked=(&self:WidgetTree.Button_Quit.HandleOnButtonClicked)
    onButtonBaseDoubleClicked=(&self:WidgetTree.Button_Quit.HandleOnButtonDoubleClicked)
    onButtonBaseHovered=(&self:WidgetTree.Button_Quit.HandleOnButtonHovered)
    onButtonBaseUnhovered=(&self:WidgetTree.Button_Quit.HandleOnButtonUnhovered)
    onSelectedChangedBase=(&self:WidgetTree.Button_Quit.HandleOnSelectedChanged)
    style=/FortUILibrary/Foundation/Buttons/ButtonStyles/ButtonStyle_Feature_M.ButtonStyle_Feature_M_C
    visibility=Collapsed
    triggeringInputAction {dataTable=&InputActionDataTable, rowName=Cancel}
   }
   slot {class=HorizontalBoxSlot, properties {padding=<ltrb 16 0 0 0>}}
  }
  HorizontalBox_Redesign {
   widgetClassPath=&HorizontalBox
   bIsVariable=true
   parent=Switcher_Button_Default_Redesign
   widget=HorizontalBox_Redesign
   slot {class=WidgetSwitcherSlot}
  }
  Button_Continue_Redesign {
   widgetClassPath=/FortUILibrary/UIKit/Buttons/WBP_UIKit_Button_Loud
   bInherited=true
   bIsVariable=true
   parent=HorizontalBox_Redesign
   widget=Button_Continue_Redesign
   properties {text=Retry, triggeringInputAction {dataTable=&InputActionDataTable, rowName=RetryLogin}}
   slot {class=HorizontalBoxSlot}
  }
  Button_Quit_Redesign {
   widgetClassPath=&WBP_UIKit_Button_Regular
   bInherited=true
   bIsVariable=true
   parent=HorizontalBox_Redesign
   widget=Button_Quit_Redesign
   properties {text=Quit, triggeringInputAction {dataTable=&InputActionDataTable, rowName=Cancel}}
   slot {class=HorizontalBoxSlot, properties {padding=<ltrb 16 0 0 0>}}
  }
  Button_Status {
   widgetClassPath=&IconTextButton
   bInherited=true
   bIsVariable=true
   parent=VerticalBox_Content
   widget=Button_Status
   properties {
    buttonText=Check Current Online Status
    contentAlignment=HAlign_Left
    minHeight=100
    navigation=&self:WidgetTree.Button_Status.WidgetNavigation_0
    onButtonBaseClicked=(&self:WidgetTree.Button_Status.HandleOnButtonClicked)
    onButtonBaseDoubleClicked=(&self:WidgetTree.Button_Status.HandleOnButtonDoubleClicked)
    onButtonBaseHovered=(&self:WidgetTree.Button_Status.HandleOnButtonHovered)
    onButtonBaseUnhovered=(&self:WidgetTree.Button_Status.HandleOnButtonUnhovered)
    onSelectedChangedBase=(&self:WidgetTree.Button_Status.HandleOnSelectedChanged)
    style=/FortUILibrary/Foundation/Buttons/ButtonStyles/ButtonStyle-TextOnlyBase_S-B.ButtonStyle-TextOnlyBase_S-B_C
    visibility=Hidden
    iconBrush {
     bIsDynamicallyLoaded=false
     drawAs=NoDrawType
     imageType=NoImage
     mirroring=NoMirror
     resourceName=None
     resourceObject=&M-Icon-ExternalLink-64
     tiling=NoTile
     imageSize=<76 76>
     margin=<ltrb 0 0 0 0>
     outlineSettings {
      bUseBrushTransparency=false
      roundingType=HalfHeightRadius
      width=0
      color {colorUseRule=UseColor_Specified, specifiedColor=<rgba 0 0 0 0>}
      cornerRadii=<0 0 0 1>
     }
     tintColor {colorUseRule=UseColor_Specified, specifiedColor=<rgba 1 1 1 1>}
     uVRegion {bIsValid=false, max=<0 0>, min=<0 0>}
    }
    triggeringInputAction {dataTable=&InputActionDataTable, rowName=GenericAccept}
   }
   slot {class=VerticalBoxSlot, properties {horizontalAlignment=HAlign_Center, verticalAlignment=VAlign_Bottom, padding=<ltrb 0 10 0 10>}}
  }
  Border_1 {
   widgetClassPath=&Border
   namedSlotHost=Button_Status
   widget=Border_1
   properties {
    horizontalAlignment=HAlign_Left
    verticalAlignment=VAlign_Center
    background {
     bIsDynamicallyLoaded=false
     drawAs=NoDrawType
     imageType=NoImage
     mirroring=NoMirror
     resourceName=None
     resourceObject=null
     tiling=NoTile
     imageSize=<32 32>
     margin=<ltrb 0 0 0 0>
     outlineSettings {
      bUseBrushTransparency=false
      roundingType=HalfHeightRadius
      width=0
      color {colorUseRule=UseColor_Specified, specifiedColor=<rgba 0 0 0 0>}
      cornerRadii=<0 0 0 0>
     }
     tintColor {colorUseRule=UseColor_Specified, specifiedColor=<rgba 1 1 1 1>}
     uVRegion {bIsValid=false, max=<0 0>, min=<0 0>}
    }
    padding=<ltrb 4 -4 0 0>
   }
   slot {class=VerticalBoxSlot}
  }
  Image_4 {
   widgetClassPath=&Image
   parent=Border_1
   widget=Image_4
   properties {brush {
     bIsDynamicallyLoaded=false
     drawAs=Image
     imageType=NoImage
     mirroring=NoMirror
     resourceName=None
     resourceObject=&M-Icon-ExternalLink-64
     tiling=NoTile
     imageSize=<70 70>
     margin=<ltrb 0 0 0 0>
     outlineSettings {
      bUseBrushTransparency=false
      roundingType=HalfHeightRadius
      width=0
      color {colorUseRule=UseColor_Specified, specifiedColor=<rgba 0 0 0 0>}
      cornerRadii=<0 0 0 0>
     }
     tintColor {colorUseRule=UseColor_Specified, specifiedColor=<rgba 1 1 1 1>}
     uVRegion {bIsValid=false, max=<0 0>, min=<0 0>}
    }}
   slot {class=BorderSlot, properties {horizontalAlignment=HAlign_Left, verticalAlignment=VAlign_Center, padding=<ltrb 4 -4 0 0>}}
  }
  SizeBox_Text_Status {
   widgetClassPath=&SizeBox
   bIsVariable=true
   parent=VerticalBox_Content
   widget=SizeBox_Text_Status
   properties {bOverride_WidthOverride=true, widthOverride=1408}
   slot {class=VerticalBoxSlot, properties {verticalAlignment=VAlign_Top}}
  }
  Text_Status {
   widgetClassPath=&CommonTextBlock
   bInherited=true
   parent=SizeBox_Text_Status
   widget=Text_Status
   properties {
    justification=Center
    renderOpacity=0.5
    wrapTextAt=1408
    font {
     bForceMonospaced=false
     bMaterialIsStencil=false
     fontMaterial=null
     fontObject=&BurbankSmall
     letterSpacing=0
     monospacedWidth=1
     size=16
     skewAmount=0
     typefaceFontName=Bold
     outlineSettings {
      bApplyOutlineToDropShadows=false
      bMiteredCorners=false
      bSeparateFillAlpha=false
      outlineMaterial=null
      outlineSize=0
      outlineColor=<rgba 0 0 0 1>
     }
    }
    shadowOffset=<0 0>
   }
   slot {class=SizeBoxSlot, properties {horizontalAlignment=HAlign_Center}}
  }
 ]
}
