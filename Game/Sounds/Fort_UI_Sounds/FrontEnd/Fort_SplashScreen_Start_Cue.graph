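# Sound cue for the front-end splash screen start sound in Fortnite UI, routed to the Fort_UI sound class.
# A mixer layers two reused Item Shop waves, a deep-impact click and a menu toggle, so they play together at equal volume.
# A concurrency override allows one instance at a time and stops the oldest when it retriggers.
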SoundCue /Game/Sounds/Fort_UI_Sounds/FrontEnd/Fort_SplashScreen_Start_Cue
root=Mixer_0
nodes [
 {id=Mixer_0, children [WavePlayer_11, WavePlayer_1], properties {inputVolume [1,1]}}
 {id=WavePlayer_1, properties {soundWaveAssetPtr=/Game/Sounds/Fort_UI_Sounds/FrontEnd/ItemShop/wav/Fort_UI_Shop_MenuToggle}}
 {id=WavePlayer_11, properties {soundWaveAssetPtr=/Game/Sounds/Fort_UI_Sounds/FrontEnd/ItemShop/wav/Fort_UI_Shop_Click_DeepImpact_01}}
]
properties {
 bOverrideConcurrency=true
 maxDistance=2097152
 soundClassObject=&Fort_UI
 volumeMultiplier=1
 concurrencyOverrides {maxCount=1, resolutionRule=StopOldest}
}
