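# Login state screen layout on FortUIStateWidget_Login, switching between splash, login, status, safe zone, health warning and other pre-lobby screens over a background.
# Activation updates the watermark, sets frontend input and login camera, fires analytics and conditionally starts splash music.
# SetKeyArtVisibility shows key art and boot-up background only for the first screen or classes in ShowBackgroundWidgets.
# Construct fills version text and pads the mobile footer; splash deactivation plays a start sound.
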WidgetBlueprint /FortniteCoreUI/Foundation/StateLayouts/LoginScreen
graphs [
 EventGraph {nodes [
   "On Widget Deactivated (Splash_Screen)"() {is=title K2Node_ComponentBoundEvent}
   > PlaySound2D(Sound=/Game/Sounds/Fort_UI_Sounds/FrontEnd/Fort_SplashScreen_Start_Cue)
   n1 = Event SetKeyArtVisibility!() -> (OutputDelegate, then, NewActiveWidgetIndex:int=0) {is=title K2Node_Event}
   n2 = GetChildAt!(
    self:Panel Widget Object Reference=Get Switcher_ScreenContents() -> (Switcher_ScreenContents:Common Activatable Widget Switcher Object Reference)
    Index:int=@n1.NewActiveWidgetIndex
   ) -> (ReturnValue:Widget Object Reference)
   n3 = Contains Item(
    TargetArray:Array of Widget Soft Class References=Get ShowBackgroundWidgets() -> (ShowBackgroundWidgets:Array of Widget Soft Class References)
    ItemToFind:"Widget Soft Class Reference (by ref)"=Make Soft Reference!(Input:Widget Class Reference=Get Class(Object=@n2) -> (ReturnValue:Widget Class Reference)) -> (Output:Widget Soft Class Reference)
   )
   n4 = "Equal (Integer)"(A=@n1.NewActiveWidgetIndex)
   > SetVisibility(
    self=[
     Get FortniteBootUpBackground() -> (FortniteBootUpBackground:Fortnite Boot Up Background Object Reference)
     Get KeyArt-Group() -> (KeyArt-Group:Overlay Object Reference)
    ]
    InVisibility=Select(Option 0=Collapsed, Option 1=SelfHitTestInvisible, Index=OR Boolean(A=@n4, B=@n3))
   )
   Event On Activated()
   > UpdateWatermark(self:Fort Build Watermark Object Reference=Get Widget_BuildWatermark() -> (Widget_BuildWatermark:Build Watermark Object Reference))
   > SetInputMode!(
    execute
    self:Fort Global UIContext Object Reference=Get FortGlobalUIContext!() -> (ReturnValue:Fort Global UIContext Object Reference) {is=title K2Node_GetBlueprintContextEx}
    InMode:EFortInputMode Enum=Frontend
   ) -> (then)
   > SetFrontEndCamera!(
    execute
    self:Fort Front End Context Object Reference=Get FortFrontEndContext!() -> (ReturnValue:Fort Front End Context Object Reference) {is=title K2Node_GetBlueprintContextEx}
    NewState:EFrontEndCamera Enum=Login
    TransitionParams:View Target Transition Params Structure=(BlendTime=0,BlendFunction=VTBlend_Cubic,BlendExp=2,bLockOutgoing=False)
   ) -> (then)
   > Fire Event Main View Opened!(execute, View:Widget Object Reference=Self-Reference(), AdditionalInfo:Flow Analytics Additional Info Structure, bIsOverlay:bool=false) -> (then)
   > Branch(Condition=ShouldPlaySplashScreenMusic!(self:Fort UIState Widget Login Object Reference) -> (ReturnValue:bool))
   > AddEvent!(
    execute
    self:Music Event Subsystem Object Reference=Get MusicEventSubsystem() -> (ReturnValue:Music Event Subsystem Object Reference) {is=title K2Node_GetSubsystem}
    Asset:Sound Base Object Reference=/Game/Athena/Sounds/MusicPacks/S28_DefaultLobbyTrack/SC_Default_LoginMusic_Cue
    EventTag:Gameplay Tag Structure=(TagName="MusicEventSubsystem.Event.SplashScreen")
    EventBehaviorTags:Gameplay Tag Container Structure=(GameplayTags=)
   ) -> (then, ReturnValue:Audio Component Object Reference)
   Event Construct()
   n5 = GetRuntimeOptions()
   > "SetText (Text)"(
    self=Get Text_Version() -> (Text_Version:Common Text Object Reference)
    InText="To Text (String)"(InString=GetGameVersion!(self:Fort Runtime Options Object Reference=@n5) -> (ReturnValue:String))
   )
   > SetCompactUIStyle() {is=title K2Node_CallFunction}
  ]}
 SetCompactUIStyle {nodes [
   SetCompactUIStyle() {is=title K2Node_FunctionEntry}
   > Branch(Condition=IsMobileGame())
   > SetPadding!(
    execute
    self:Border Slot Object Reference=SlotAsBorderSlot!(Widget:Widget Object Reference=Get Footer() -> (Footer:Vertical Box Object Reference)) -> (ReturnValue:Border Slot Object Reference)
    InPadding_Left:float=64
    InPadding_Top:float=0
    InPadding_Right:float=64
    InPadding_Bottom:float=0
   ) -> (then)
  ]}
]
variables [PlatformSpecificBackBorder {type=Widget, default=null}]
widgets {
 info {inheritedWidgetCount=14, parentClass=/Script/FortniteUI.FortUIStateWidget_Login, rootWidgetClass=&Overlay, widgetCount=39}
 widgets [
  Overlay_4 {widgetClassPath=&Overlay, widget=Overlay_4}
  Switcher_Background {
   widgetClassPath=&CommonActivatableWidgetSwitcher
   bInherited=true
   bIsVariable=true
   parent=Overlay_4
   widget=Switcher_Background
   slot {class=OverlaySlot, properties {horizontalAlignment=HAlign_Fill, verticalAlignment=VAlign_Fill}}
  }
  Overlay_DefaultBackground {
   widgetClassPath=&Overlay
   bInherited=true
   parent=Switcher_Background
   widget=Overlay_DefaultBackground
   slot {class=WidgetSwitcherSlot}
  }
  ScaleBox {
   widgetClassPath=&ScaleBox
   parent=Overlay_DefaultBackground
   widget=ScaleBox
   properties {stretch=Fill}
   slot {class=OverlaySlot, properties {horizontalAlignment=HAlign_Fill, verticalAlignment=VAlign_Fill}}
  }
  SizeBox_2 {
   widgetClassPath=&SizeBox
   parent=ScaleBox
   widget=SizeBox_2
   slot {class=ScaleBoxSlot, properties {horizontalAlignment=HAlign_Fill}}
  }
  Overlay_0 {widgetClassPath=&Overlay, parent=SizeBox_2, widget=Overlay_0, slot {class=SizeBoxSlot}}
  BackgroundGradient {
   widgetClassPath=&Image
   parent=Overlay_0
   widget=BackgroundGradient
   properties {
    visibility=SelfHitTestInvisible
    brush {
     bIsDynamicallyLoaded=false
     drawAs=Image
     imageType=NoImage
     mirroring=NoMirror
     resourceName=None
     resourceObject=/Game/Athena/UI/Frontend/Art/M_UI_RadialGradientMaster_Inst
     tiling=NoTile
     imageSize=<1920 1080>
     margin=<ltrb 0 0 0 0>
     outlineSettings {
      bUseBrushTransparency=false
      roundingType=HalfHeightRadius
      width=0
      color {colorUseRule=UseColor_Specified, specifiedColor=<rgba 0 0 0 0>}
      cornerRadii=<0 0 0 0>
     }
     tintColor {colorUseRule=UseColor_Specified, specifiedColor=<rgba 1 1 1 1>}
     uVRegion {bIsValid=false, max=<0 0>, min=<0 0>}
    }
   }
   slot {class=OverlaySlot, properties {horizontalAlignment=HAlign_Fill, verticalAlignment=VAlign_Fill}}
  }
  KeyArt-Group {
   widgetClassPath=&Overlay
   bIsVariable=true
   parent=Overlay_0
   widget=KeyArt-Group
   slot {class=OverlaySlot, properties {horizontalAlignment=HAlign_Fill, verticalAlignment=VAlign_Fill}}
  }
  DarkBacking {
   widgetClassPath=&Image
   bIsVariable=true
   parent=KeyArt-Group
   widget=DarkBacking
   properties {brush {
     bIsDynamicallyLoaded=false
     drawAs=Image
     imageType=NoImage
     mirroring=NoMirror
     resourceName=None
     resourceObject=null
     tiling=NoTile
     imageSize=<1920 1080>
     margin=<ltrb 0 0 0 0>
     outlineSettings {
      bUseBrushTransparency=false
      roundingType=HalfHeightRadius
      width=0
      color {colorUseRule=UseColor_Specified, specifiedColor=<rgba 0 0 0 0>}
      cornerRadii=<0 0 0 0>
     }
     tintColor {colorUseRule=UseColor_Specified, specifiedColor=<rgba 0 0 0 1>}
     uVRegion {bIsValid=false, max=<0 0>, min=<0 0>}
    }}
   slot {class=OverlaySlot, properties {horizontalAlignment=HAlign_Fill, verticalAlignment=VAlign_Fill}}
  }
  KeyArtBackground {
   widgetClassPath=&Image
   bIsVariable=true
   parent=KeyArt-Group
   widget=KeyArtBackground
   properties {
    visibility=SelfHitTestInvisible
    brush {
     bIsDynamicallyLoaded=false
     drawAs=Image
     imageType=NoImage
     mirroring=NoMirror
     resourceName=None
     resourceObject=/Game/UI/Foundation/Textures/LoadingScreens/Fort_Load_SeasonKeyArt
     tiling=NoTile
     imageSize=<1920 1080>
     margin=<ltrb 0 0 0 0>
     outlineSettings {
      bUseBrushTransparency=false
      roundingType=HalfHeightRadius
      width=0
      color {colorUseRule=UseColor_Specified, specifiedColor=<rgba 0 0 0 0>}
      cornerRadii=<0 0 0 0>
     }
     tintColor {colorUseRule=UseColor_Specified, specifiedColor=<rgba 1 1 1 1>}
     uVRegion {bIsValid=false, max=<0 0>, min=<0 0>}
    }
   }
   slot {class=OverlaySlot, properties {horizontalAlignment=HAlign_Center, verticalAlignment=VAlign_Center}}
  }
  Overlay_RedesignBackground {
   widgetClassPath=&Overlay
   bInherited=true
   parent=Switcher_Background
   widget=Overlay_RedesignBackground
   slot {class=WidgetSwitcherSlot}
  }
  BackgroundGradient_Redesign {
   widgetClassPath=&Image
   parent=Overlay_RedesignBackground
   widget=BackgroundGradient_Redesign
   properties {
    visibility=SelfHitTestInvisible
    brush {
     bIsDynamicallyLoaded=false
     drawAs=Image
     imageType=NoImage
     mirroring=NoMirror
     resourceName=None
     resourceObject=&M_UI_RadialGradientMaster_Inst
     tiling=NoTile
     imageSize=<1920 1080>
     margin=<ltrb 0 0 0 0>
     outlineSettings {
      bUseBrushTransparency=false
      roundingType=HalfHeightRadius
      width=0
      color {colorUseRule=UseColor_Specified, specifiedColor=<rgba 0 0 0 0>}
      cornerRadii=<0 0 0 0>
     }
     tintColor {colorUseRule=UseColor_Specified, specifiedColor=<rgba 1 1 1 1>}
     uVRegion {bIsValid=false, max=<0 0>, min=<0 0>}
    }
   }
   slot {class=OverlaySlot, properties {horizontalAlignment=HAlign_Fill, verticalAlignment=VAlign_Fill}}
  }
  FortniteBootUpBackground {
   widgetClassPath=/Script/FortniteEarlyPreloadWidgets.FortniteBootUpBackground
   bIsVariable=true
   parent=Overlay_RedesignBackground
   widget=FortniteBootUpBackground
   slot {class=OverlaySlot, properties {horizontalAlignment=HAlign_Fill, verticalAlignment=VAlign_Fill}}
  }
  CommonVisibilityWidget_0 {
   widgetClassPath=&CommonVisibilityWidgetLegacy
   parent=Overlay_4
   widget=CommonVisibilityWidget_0
   properties {
    bShowForAndroid=false
    bShowForIOS=false
    bShowForMac=false
    bShowForPS4=false
    bShowForPS5=false
    bShowForXBox=false
    bShowForXSX=false
    verticalAlignment=VAlign_Top
    padding=<ltrb 0 0 0 0>
   }
   slot {class=OverlaySlot, properties {horizontalAlignment=HAlign_Fill}}
  }
  SizeBox_0 {
   widgetClassPath=&SizeBox
   parent=CommonVisibilityWidget_0
   widget=SizeBox_0
   properties {bOverride_HeightOverride=true, heightOverride=64}
   slot {class=BorderSlot, properties {verticalAlignment=VAlign_Top, padding=<ltrb 0 0 0 0>}}
  }
  WindowTitleBar {
   widgetClassPath=/Script/UMG.WindowTitleBarArea
   parent=SizeBox_0
   widget=WindowTitleBar
   properties {bWindowButtonsEnabled=true}
   slot {class=SizeBoxSlot}
  }
  SafeZone {
   widgetClassPath=&SafeZone
   bIsVariable=true
   parent=Overlay_4
   widget=SafeZone
   slot {class=OverlaySlot, properties {horizontalAlignment=HAlign_Fill, verticalAlignment=VAlign_Fill}}
  }
  Border_2 {
   widgetClassPath=&Border
   parent=SafeZone
   widget=Border_2
   properties {
    background {
     bIsDynamicallyLoaded=false
     drawAs=NoDrawType
     imageType=NoImage
     mirroring=NoMirror
     resourceName=None
     resourceObject=null
     tiling=NoTile
     imageSize=<32 32>
     margin=<ltrb 0 0 0 0>
     outlineSettings {
      bUseBrushTransparency=false
      roundingType=HalfHeightRadius
      width=0
      color {colorUseRule=UseColor_Specified, specifiedColor=<rgba 0 0 0 0>}
      cornerRadii=<0 0 0 0>
     }
     tintColor {colorUseRule=UseColor_Specified, specifiedColor=<rgba 1 1 1 1>}
     uVRegion {bIsValid=false, max=<0 0>, min=<0 0>}
    }
    contentColorAndOpacity=<rgba 0.5 0.5 0.5 1>
    padding=<ltrb 0 0 0 0>
   }
   slot {class=SafeZoneSlot, properties {vAlign=VAlign_Bottom}}
  }
  Footer {
   widgetClassPath=&VerticalBox
   bIsVariable=true
   parent=Border_2
   widget=Footer
   slot {class=BorderSlot, properties {padding=<ltrb 0 0 0 0>}}
  }
  Button_Cancel {
   widgetClassPath=&IconTextButton
   bInherited=true
   bIsVariable=true
   parent=Footer
   widget=Button_Cancel
   properties {
    buttonText=Close
    onButtonBaseClicked=(&self:WidgetTree.Button_Cancel.HandleOnButtonClicked)
    onButtonBaseDoubleClicked=(&self:WidgetTree.Button_Cancel.HandleOnButtonDoubleClicked)
    onButtonBaseHovered=(&self:WidgetTree.Button_Cancel.HandleOnButtonHovered)
    onButtonBaseUnhovered=(&self:WidgetTree.Button_Cancel.HandleOnButtonUnhovered)
    onSelectedChangedBase=(&self:WidgetTree.Button_Cancel.HandleOnSelectedChanged)
    style=/FortUILibrary/Foundation/Buttons/ButtonStyles/Primary/ButtonStyle-Primary-M_Skew_Yellow.ButtonStyle-Primary-M_Skew_Yellow_C
    triggeringInputAction {dataTable=&InputActionDataTable, rowName=CloseToHold}
   }
   slot {class=VerticalBoxSlot, properties {horizontalAlignment=HAlign_Right, verticalAlignment=VAlign_Bottom, padding=<ltrb 20 20 20 20>}}
  }
  HorizontalBox_3 {widgetClassPath=&HorizontalBox, parent=Footer, widget=HorizontalBox_3, slot {class=VerticalBoxSlot}}
  VerticalBox_0 {
   widgetClassPath=&VerticalBox
   parent=HorizontalBox_3
   widget=VerticalBox_0
   slot {class=HorizontalBoxSlot, properties {verticalAlignment=VAlign_Bottom, size {sizeRule=Fill, value=1}}}
  }
  ConsoleProfileWidget {
   widgetClassPath=/FortniteCoreUI/UI/Foundation/Profile/ConsoleProfileWidget
   bIsVariable=true
   parent=VerticalBox_0
   widget=ConsoleProfileWidget
   slot {class=VerticalBoxSlot, properties {horizontalAlignment=HAlign_Left}}
  }
  HorizontalBox_0 {
   widgetClassPath=&HorizontalBox
   parent=VerticalBox_0
   widget=HorizontalBox_0
   slot {class=VerticalBoxSlot, properties {horizontalAlignment=HAlign_Left, verticalAlignment=VAlign_Bottom, padding=<ltrb 10 0 0 0>}}
  }
  Text_Version {
   widgetClassPath=&CommonTextBlock
   bIsVariable=true
   parent=HorizontalBox_0
   widget=Text_Version
   properties {
    text=Build Version
    font {
     bForceMonospaced=false
     bMaterialIsStencil=false
     fontMaterial=null
     fontObject=&NotoSans
     letterSpacing=0
     monospacedWidth=1
     size=10
     skewAmount=0
     typefaceFontName=NotoSans-Bold
     outlineSettings {
      bApplyOutlineToDropShadows=false
      bMiteredCorners=false
      bSeparateFillAlpha=false
      outlineMaterial=null
      outlineSize=0
      outlineColor=<rgba 0 0 0 1>
     }
    }
   }
   slot {class=HorizontalBoxSlot}
  }
  Widget_BuildWatermark {
   widgetClassPath=/Game/UI/Foundation/Watermark/BuildWatermark
   bIsVariable=true
   parent=HorizontalBox_0
   widget=Widget_BuildWatermark
   slot {class=HorizontalBoxSlot, properties {horizontalAlignment=HAlign_Left, verticalAlignment=VAlign_Top, padding=<ltrb 10 0 0 0>}}
  }
  UnrealLogo {
   widgetClassPath=&Image
   parent=HorizontalBox_3
   widget=UnrealLogo
   properties {
    visibility=SelfHitTestInvisible
    brush {
     bIsDynamicallyLoaded=false
     drawAs=Image
     imageType=NoImage
     mirroring=NoMirror
     resourceName=None
     resourceObject=/FortniteCoreUI/UI/Foundation/Textures/Icons/Social/T-Icon-UnrealEngine5-64
     tiling=NoTile
     imageSize=<64 64>
     margin=<ltrb 0 0 0 0>
     outlineSettings {
      bUseBrushTransparency=false
      roundingType=HalfHeightRadius
      width=0
      color {colorUseRule=UseColor_Specified, specifiedColor=<rgba 0 0 0 0>}
      cornerRadii=<0 0 0 0>
     }
     tintColor {colorUseRule=UseColor_Specified, specifiedColor=<rgba 1 1 1 1>}
     uVRegion {bIsValid=false, max=<0 0>, min=<0 0>}
    }
   }
   slot {class=HorizontalBoxSlot, properties {horizontalAlignment=HAlign_Center, verticalAlignment=VAlign_Center}}
  }
  EGLogo {
   widgetClassPath=&Image
   parent=HorizontalBox_3
   widget=EGLogo
   properties {
    visibility=SelfHitTestInvisible
    brush {
     bIsDynamicallyLoaded=false
     drawAs=Image
     imageType=NoImage
     mirroring=NoMirror
     resourceName=None
     resourceObject=/Game/UI/Foundation/Textures/Icons/Social/T-Icon-EpicGamesLogo-64
     tiling=NoTile
     imageSize=<56 56>
     margin=<ltrb 0 0 0 0>
     outlineSettings {
      bUseBrushTransparency=false
      roundingType=HalfHeightRadius
      width=0
      color {colorUseRule=UseColor_Specified, specifiedColor=<rgba 0 0 0 0>}
      cornerRadii=<0 0 0 0>
     }
     tintColor {colorUseRule=UseColor_Specified, specifiedColor=<rgba 1 1 1 1>}
     uVRegion {bIsValid=false, max=<0 0>, min=<0 0>}
    }
   }
   slot {class=HorizontalBoxSlot, properties {horizontalAlignment=HAlign_Center, verticalAlignment=VAlign_Center}}
  }
  ActionBar_Bottom {
   widgetClassPath=/Game/UI/Frontend/BottomBar/WBP_AthenaBottomBarWidget
   bInherited=true
   bIsVariable=true
   parent=Footer
   widget=ActionBar_Bottom
   properties {display Owner Actions Only=false}
   slot {class=VerticalBoxSlot}
  }
  Switcher_ScreenContents {
   widgetClassPath=&CommonActivatableWidgetSwitcher
   bInherited=true
   bIsVariable=true
   parent=Overlay_4
   widget=Switcher_ScreenContents
   slot {class=OverlaySlot, properties {horizontalAlignment=HAlign_Fill, verticalAlignment=VAlign_Fill}}
  }
  EmptyScreen_Default {widgetClassPath=&Spacer, parent=Switcher_ScreenContents, widget=EmptyScreen_Default, slot {class=WidgetSwitcherSlot}}
  Login_Interactions {
   widgetClassPath=/FortniteCoreUI/UI/Login/LoginInteractionHost
   bInherited=true
   bIsVariable=true
   parent=Switcher_ScreenContents
   widget=Login_Interactions
   properties {bIsFocusable=true}
   slot {class=WidgetSwitcherSlot}
  }
  Status_LoginInfo {
   widgetClassPath=/FortniteCoreUI/UI/Login/Status/StatusWidget
   bInherited=true
   bIsVariable=true
   parent=Switcher_ScreenContents
   widget=Status_LoginInfo
   properties {bIsFocusable=true}
   slot {class=WidgetSwitcherSlot}
  }
  Splash_Screen {
   widgetClassPath=/FortniteCoreUI/UI/Login/SplashScreen/SplashScreenWidget
   bInherited=true
   bIsVariable=true
   parent=Switcher_ScreenContents
   widget=Splash_Screen
   slot {class=WidgetSwitcherSlot}
  }
  Editor_SafeZone {
   widgetClassPath=/FortniteCoreUI/UI/Login/SafeZone/SafeZoneEditor
   bInherited=true
   bIsVariable=true
   parent=Switcher_ScreenContents
   widget=Editor_SafeZone
   properties {bCanCancel=false}
   slot {class=WidgetSwitcherSlot}
  }
  HealthWarningScreen_HealthInfo {
   widgetClassPath=/FortniteCoreUI/UI/Login/HealthWarning/HealthWarningScreen
   bInherited=true
   bIsVariable=true
   parent=Switcher_ScreenContents
   widget=HealthWarningScreen_HealthInfo
   slot {class=WidgetSwitcherSlot}
  }
  PushNotificationSelectScreen {
   widgetClassPath=/FortniteCoreUI/UI/Login/PushNotifications/PushNotificationSelectionScreen
   bInherited=true
   bIsVariable=true
   parent=Switcher_ScreenContents
   widget=PushNotificationSelectScreen
   slot {class=WidgetSwitcherSlot}
  }
  PostLoginHoldingTankScreen {
   widgetClassPath=/FortniteCoreUI/UI/Login/PostLoginHoldingTank/PostLoginHoldingTank
   bInherited=true
   bIsVariable=true
   parent=Switcher_ScreenContents
   widget=PostLoginHoldingTankScreen
   slot {class=WidgetSwitcherSlot}
  }
  WidgetStack_PrimaryContent {widgetClassPath=/Script/CommonUI.CommonActivatableWidgetStack, bInherited=true, widget=null}
 ]
}
